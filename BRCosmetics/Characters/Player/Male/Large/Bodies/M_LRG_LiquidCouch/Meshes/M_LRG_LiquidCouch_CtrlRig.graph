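# Control Rig for the LiquidCouch large male player body that procedurally corrects robot arms, shoulder pads, pouches, straps and props.
# It runs DeformRig, legs, pouches and arms, then shoulderpads (skipped in FE idle) and spine.
# DeltaRotation v4 remaps swing or twist against construction-created nulls into corrective weights, some multiplied for combined poses.
# Constraints blend attachments, Hermite splines drive chest tubes, and smoothed targeting shifts the right shoulderpad.
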ControlRigBlueprint /BRCosmetics/Characters/Player/Male/Large/Bodies/M_LRG_LiquidCouch/Meshes/M_LRG_LiquidCouch_CtrlRig
graphs [
 New Function_ContainedGraph {nodes [
   RigVMFunctionEntryNode()
   > &RigUnit_ParentConstraint(
    AdvancedSettings=(InterpolationType=Average,RotationOrderForFilter=XZY)
    Child=(Type=Bone,Name="MSTR_pouch_01_l")
    Filter=(TranslationFilter=(bX=false,bY=false,bZ=false),RotationFilter=(bX=true,bY=True,bZ=True),ScaleFilter=(bX=false,bY=false,bZ=false))
    Parents=((Item=(Type=Bone,Name="thigh_l"),Weight=0.8),(Item=(Type=Bone,Name="pelvis"),Weight=0.2))
    Weight=1
   )
   > &RigUnit_ParentConstraint(
    AdvancedSettings=(InterpolationType=Average,RotationOrderForFilter=XZY)
    Child=(Type=Bone,Name="MSTR_pouch_01_r")
    Filter=(TranslationFilter=(bX=false,bY=false,bZ=false),RotationFilter=(bX=true,bY=True,bZ=True),ScaleFilter=(bX=false,bY=false,bZ=false))
    Parents=((Item=(Type=Bone,Name="thigh_r"),Weight=0.8),(Item=(Type=Bone,Name="pelvis"),Weight=0.2))
    Weight=1
   )
   > DeltaRotation v4 = RigVMFunctionReferenceNode(
    Rotation Order=XYZ
    child=(Type=Bone,Name="thigh_l")
    debugRotation=false
    outputMinMax=(X=0,Y=1)
    parent=(Type=Null,Name="thigh_l_null")
    premult=1
    rotationMinMax=(X=-60,Y=-100)
    triggerAxis=Z
    twistAxis=(X=1,Y=0,Z=0)
   )
   > &RigUnit_ModifyTransforms(
    ItemToModify=((Item=(Type=Bone,Name="MSTR_pouch_01_l"),Transform=(Rotation=(X=-0.0545,Y=-0.1853,Z=0.1555,W=0.9688),Translation=(X=2.6134,Y=0.2288,Z=0.7034),Scale3D=(X=1,Y=1,Z=1))))
    WeightMinimum=0
    Weight=@"DeltaRotation v4".outputWeight
   )
   > DeltaRotation v4_1 = RigVMFunctionReferenceNode(
    Rotation Order=XYZ
    child=(Type=Bone,Name="thigh_r")
    debugRotation=false
    outputMinMax=(X=0,Y=1)
    parent=(Type=Null,Name="thigh_r_null")
    premult=1
    rotationMinMax=(X=-60,Y=-100)
    triggerAxis=Z
    twistAxis=(X=1,Y=0,Z=0)
   )
   > &RigUnit_ModifyTransforms(
    ItemToModify="((Item=(Type=Bone,Name=\"MSTR_pouch_01_r\"),Transform=(Rotation=(X=-0.0439,Y=-0.1811,Z=0.1247,W=0.9745),Translation=(X=0,Y=0,Z=0),Scale3D=(X=1,Y=1,Z=1))),(Item=(Type=Bone,Name=\"MSTR_wrench\"),Transform=(Rotation=(X=0.0781,Y=0.1773,Z=-0.0307,W=0.9806),Translat ...(truncated)"
    WeightMinimum=0
    Weight=@"DeltaRotation v4_1".outputWeight
   )
   > DeltaRotation v4_2 = RigVMFunctionReferenceNode(
    Rotation Order=XYZ
    child=(Type=Bone,Name="thigh_l")
    debugRotation=false
    outputMinMax=(X=0,Y=1)
    parent=(Type=Null,Name="thigh_l_null")
    premult=1
    rotationMinMax=(X=0,Y=75)
    triggerAxis=Z
    twistAxis=(X=1,Y=0,Z=0)
   )
   > &RigUnit_ModifyTransforms(
    ItemToModify=((Item=(Type=Bone,Name="MSTR_pouch_01_l"),Transform=(Rotation=(X=0.0072,Y=-0.2503,Z=0.3654,W=0.8965),Translation=(X=0,Y=0,Z=0),Scale3D=(X=1,Y=1,Z=1))))
    WeightMinimum=0
    Weight=@"DeltaRotation v4_2".outputWeight
   )
   > DeltaRotation v4_3 = RigVMFunctionReferenceNode(
    Rotation Order=XYZ
    child=(Type=Bone,Name="thigh_r")
    debugRotation=false
    outputMinMax=(X=0,Y=1)
    parent=(Type=Null,Name="thigh_r_null")
    premult=1
    rotationMinMax=(X=0,Y=75)
    triggerAxis=Z
    twistAxis=(X=1,Y=0,Z=0)
   )
   > &RigUnit_ModifyTransforms(
    ItemToModify=((Item=(Type=Bone,Name="MSTR_pouch_01_r"),Transform=(Rotation=(X=0.0072,Y=-0.2503,Z=0.3654,W=0.8965),Translation=(X=0,Y=0,Z=0),Scale3D=(X=1,Y=1,Z=1))))
    WeightMinimum=0
    Weight=@"DeltaRotation v4_3".outputWeight
   )
   > DeltaRotation v4_4 = RigVMFunctionReferenceNode(
    Rotation Order=XYZ
    child=(Type=Bone,Name="thigh_l")
    debugRotation=false
    outputMinMax=(X=0,Y=1)
    parent=(Type=Null,Name="thigh_l_null")
    premult=1
    rotationMinMax=(X=0,Y=30)
    triggerAxis=Y
    twistAxis=(X=1,Y=0,Z=0)
   )
   > &RigUnit_ModifyTransforms(
    ItemToModify=((Item=(Type=Bone,Name="MSTR_pouch_01_l"),Transform=(Rotation=(X=0.0241,Y=-0.1478,Z=-0.1228,W=0.9811),Translation=(X=0,Y=0,Z=0),Scale3D=(X=1,Y=1,Z=1))))
    WeightMinimum=0
    Weight=@"DeltaRotation v4_4".outputWeight
   )
   > DeltaRotation v4_5 = RigVMFunctionReferenceNode(
    Rotation Order=XYZ
    child=(Type=Bone,Name="thigh_r")
    debugRotation=false
    outputMinMax=(X=0,Y=1)
    parent=(Type=Null,Name="thigh_r_null")
    premult=1
    rotationMinMax=(X=0,Y=30)
    triggerAxis=Y
    twistAxis=(X=1,Y=0,Z=0)
   )
   > &RigUnit_ModifyTransforms(
    ItemToModify=((Item=(Type=Bone,Name="MSTR_pouch_01_r"),Transform=(Rotation=(X=0.0241,Y=-0.1478,Z=-0.1228,W=0.9811),Translation=(X=0,Y=0,Z=0),Scale3D=(X=1,Y=1,Z=1))))
    WeightMinimum=0
    Weight=@"DeltaRotation v4_5".outputWeight
   )
   > DeltaRotation v4_10 = RigVMFunctionReferenceNode(
    Rotation Order=XYZ
    child=(Type=Bone,Name="thigh_l")
    debugRotation=false
    outputMinMax=(X=0,Y=1)
    parent=(Type=Null,Name="thigh_l_null")
    premult=1
    rotationMinMax=(X=0,Y=-90)
    triggerAxis=Y
    twistAxis=(X=1,Y=0,Z=0)
   )
   > &RigUnit_ModifyTransforms(
    ItemToModify=((Item=(Type=Bone,Name="MSTR_pouch_01_l"),Transform=(Rotation=(X=-0.0068,Y=-0.0519,Z=-0.1303,W=0.9901),Translation=(X=0,Y=0,Z=0),Scale3D=(X=1,Y=1,Z=1))))
    WeightMinimum=0
    Weight=@"DeltaRotation v4_10".outputWeight
   )
   > DeltaRotation v4_11 = RigVMFunctionReferenceNode(
    Rotation Order=XYZ
    child=(Type=Bone,Name="thigh_r")
    debugRotation=false
    outputMinMax=(X=0,Y=1)
    parent=(Type=Null,Name="thigh_r_null")
    premult=1
    rotationMinMax=(X=0,Y=-90)
    triggerAxis=Y
    twistAxis=(X=1,Y=0,Z=0)
   )
   > &RigUnit_ModifyTransforms(
    ItemToModify=((Item=(Type=Bone,Name="MSTR_pouch_01_r"),Transform=(Rotation=(X=-0.0068,Y=-0.0519,Z=-0.1303,W=0.9901),Translation=(X=0,Y=0,Z=0),Scale3D=(X=1,Y=1,Z=1))))
    WeightMinimum=0
    Weight=@"DeltaRotation v4_11".outputWeight
   )
   > DeltaRotation v4_6 = RigVMFunctionReferenceNode(
    Rotation Order=XYZ
    child=(Type=Bone,Name="thigh_l")
    debugRotation=false
    onlyTwist=true
    outputMinMax=(X=0,Y=1)
    parent=(Type=Null,Name="thigh_l_null")
    premult=1
    rotationMinMax=(X=0,Y=75)
    triggerAxis=X
    twistAxis=(X=1,Y=0,Z=0)
   )
   > &RigUnit_ModifyTransforms(
    ItemToModify=((Item=(Type=Bone,Name="MSTR_pouch_01_l"),Transform=(Rotation=(X=0.4226,Y=0,Z=0,W=0.9063),Translation=(X=0,Y=0,Z=0),Scale3D=(X=1,Y=1,Z=1))))
    WeightMinimum=0
    Weight=@"DeltaRotation v4_6".outputWeight
   )
   > DeltaRotation v4_7 = RigVMFunctionReferenceNode(
    Rotation Order=XYZ
    child=(Type=Bone,Name="thigh_r")
    debugRotation=false
    onlyTwist=true
    outputMinMax=(X=0,Y=1)
    parent=(Type=Null,Name="thigh_r_null")
    premult=1
    rotationMinMax=(X=0,Y=75)
    triggerAxis=X
    twistAxis=(X=1,Y=0,Z=0)
   )
   > &RigUnit_ModifyTransforms(
    ItemToModify=((Item=(Type=Bone,Name="MSTR_pouch_01_r"),Transform=(Rotation=(X=0.4226,Y=0,Z=0,W=0.9063),Translation=(X=0,Y=0,Z=0),Scale3D=(X=1,Y=1,Z=1))))
    WeightMinimum=0
    Weight=@"DeltaRotation v4_7".outputWeight
   )
   > DeltaRotation v4_8 = RigVMFunctionReferenceNode(
    Rotation Order=XYZ
    child=(Type=Bone,Name="thigh_l")
    debugRotation=false
    onlyTwist=true
    outputMinMax=(X=0,Y=1)
    parent=(Type=Null,Name="thigh_l_null")
    premult=1
    rotationMinMax=(X=0,Y=-75)
    triggerAxis=X
    twistAxis=(X=1,Y=0,Z=0)
   )
   > &RigUnit_ModifyTransforms(
    ItemToModify=((Item=(Type=Bone,Name="MSTR_pouch_01_l"),Transform=(Rotation=(X=-0.421,Y=-0.079,Z=-0.0368,W=0.9029),Translation=(X=0,Y=0,Z=0),Scale3D=(X=1,Y=1,Z=1))))
    WeightMinimum=0
    Weight=@"DeltaRotation v4_8".outputWeight
   )
   > DeltaRotation v4_9 = RigVMFunctionReferenceNode(
    Rotation Order=XYZ
    child=(Type=Bone,Name="thigh_r")
    debugRotation=false
    onlyTwist=true
    outputMinMax=(X=0,Y=1)
    parent=(Type=Null,Name="thigh_r_null")
    premult=1
    rotationMinMax=(X=0,Y=-75)
    triggerAxis=X
    twistAxis=(X=1,Y=0,Z=0)
   )
   > &RigUnit_ModifyTransforms(
    ItemToModify=((Item=(Type=Bone,Name="MSTR_pouch_01_r"),Transform=(Rotation=(X=-0.421,Y=-0.079,Z=-0.0368,W=0.9029),Translation=(X=0,Y=0,Z=0),Scale3D=(X=1,Y=1,Z=1))))
    WeightMinimum=0
    Weight=@"DeltaRotation v4_9".outputWeight
   )
   Multiply = &RigVMFunction_MathDoubleMul(A=@"DeltaRotation v4".outputWeight, B=@"DeltaRotation v4_10".outputWeight)
   > &RigUnit_ModifyTransforms(
    ItemToModify=((Item=(Type=Bone,Name="MSTR_pouch_01_l"),Transform=(Rotation=(X=-0.0864,Y=-0.2473,Z=-0.1787,W=0.9484),Translation=(X=-5.3065,Y=1.9314,Z=0),Scale3D=(X=1,Y=1,Z=1))))
    WeightMinimum=0
    Weight=@Multiply.Result
   )
   > &RigUnit_ModifyTransforms(
    ItemToModify=((Item=(Type=Bone,Name="MSTR_pouch_01_r"),Transform=(Rotation=(X=-0.067,Y=-0.25,Z=-0.25,W=0.933),Translation=(X=5.3064,Y=-1.9314,Z=0),Scale3D=(X=1,Y=1,Z=1))))
    WeightMinimum=0
    Weight=&RigVMFunction_MathDoubleMul(A=@"DeltaRotation v4_1".outputWeight, B=@"DeltaRotation v4_11".outputWeight).Result
   )
   > &RigUnit_ModifyTransforms(
    ItemToModify=((Item=(Type=Bone,Name="MSTR_pouch_01_l"),Transform=(Rotation=(X=0.0084,Y=-0.1781,Z=-0.0682,W=0.9816),Translation=(X=0,Y=0,Z=0),Scale3D=(X=1,Y=1,Z=1))))
    WeightMinimum=0
    Weight=&RigVMFunction_MathDoubleMul(B=@"DeltaRotation v4_10".outputWeight, A=@"DeltaRotation v4_6".outputWeight).Result
   )
   > &RigUnit_ModifyTransforms(
    ItemToModify=((Item=(Type=Bone,Name="MSTR_pouch_01_r"),Transform=(Rotation=(X=0.0084,Y=-0.1781,Z=-0.0682,W=0.9816),Translation=(X=0,Y=0,Z=0),Scale3D=(X=1,Y=1,Z=1))))
    WeightMinimum=0
    Weight=&RigVMFunction_MathDoubleMul(B=@"DeltaRotation v4_11".outputWeight, A=@"DeltaRotation v4_9".outputWeight).Result
   )
   > &RigUnit_ModifyTransforms(
    ItemToModify=((Item=(Type=Bone,Name="MSTR_pouch_01_l"),Transform=(Rotation=(X=-0.0076,Y=-0.0868,Z=-0.0868,W=0.9924),Translation=(X=0,Y=0,Z=0),Scale3D=(X=1,Y=1,Z=1))))
    WeightMinimum=0
    Weight=&RigVMFunction_MathDoubleMul(B=@"DeltaRotation v4_4".outputWeight, A=@"DeltaRotation v4_8".outputWeight).Result
   )
   > &RigUnit_ModifyTransforms(
    ItemToModify=((Item=(Type=Bone,Name="MSTR_pouch_01_r"),Transform=(Rotation=(X=-0.0076,Y=-0.0868,Z=-0.0868,W=0.9924),Translation=(X=0,Y=0,Z=0),Scale3D=(X=1,Y=1,Z=1))))
    WeightMinimum=0
    Weight=&RigVMFunction_MathDoubleMul(B=@"DeltaRotation v4_5".outputWeight, A=@"DeltaRotation v4_9".outputWeight).Result
   )
   > RigVMFunctionReturnNode()
  ]}
 New Function_ContainedGraph {nodes [
   n1 = RigVMFunctionEntryNode()
   RigVMFunction_MathBoolToInteger = &RigVMFunction_MathBoolToInteger(Value=@n1.onlyTwist)
   RigVMFunction_MathQuaternionSwingTwist = &RigVMFunction_MathQuaternionSwingTwist(
    twistAxis=@n1.twistAxis
    Input=&RigUnit_GetRelativeTransformForItem(bChildInitial=False, bParentInitial=False, child=@n1.child, parent=@n1.parent).RelativeTransform.Rotation
   )
   n2 = RigVMDispatchNode(
    Values=((),())
    Index=@RigVMFunction_MathBoolToInteger.Result
    Values.0=@RigVMFunction_MathQuaternionSwingTwist.Swing
    Values.1=@RigVMFunction_MathQuaternionSwingTwist.Twist
   ) {template="DISPATCH_RigVMDispatch_SelectInt32(in Index,in Values,out Result)"}
   RigVMFunction_MathQuaternionToEuler = &RigVMFunction_MathQuaternionToEuler(Value=@n2.Result, RotationOrder=@n1.Rotation Order)
   > DISPATCH_RigVMDispatch_Print = RigVMDispatchNode(ScreenColor=(R=1,G=1,B=1,A=1), ScreenDuration=1.5, Enabled=@n1.debugRotation, Value=@RigVMFunction_MathQuaternionToEuler.Result) {template="DISPATCH_RigVMDispatch_Print(in Prefix,in Value,in Enabled,in ScreenDuration,in ScreenColor)"}
   Cast = RigVMDispatchNode(Value=@n1.triggerAxis) {template="DISPATCH_RigVMDispatch_CastEnumToInt(in Value,out Result)"}
   DISPATCH_RigVMDispatch_SelectInt32 = RigVMDispatchNode(
    Values=(0,0,0)
    Index=@Cast.Result
    Values.0=@RigVMFunction_MathQuaternionToEuler.Result.X
    Values.1=@RigVMFunction_MathQuaternionToEuler.Result.Y
    Values.2=@RigVMFunction_MathQuaternionToEuler.Result.Z
   ) {template="DISPATCH_RigVMDispatch_SelectInt32(in Index,in Values,out Result)"}
   Remap = &RigVMFunction_MathDoubleRemap(
    bClamp=False
    Value=@DISPATCH_RigVMDispatch_SelectInt32.Result
    TargetMinimum=@n1.outputMinMax.X
    TargetMaximum=@n1.outputMinMax.Y
    SourceMinimum=@n1.rotationMinMax.X
    SourceMaximum=@n1.rotationMinMax.Y
   )
   Absolute = &RigVMFunction_MathDoubleAbs(Value=@Remap.Result)
   Multiply = &RigVMFunction_MathDoubleMul(
    A=&RigVMFunction_MathDoubleClamp(Value=RigVMDispatchNode(True=@Absolute.Result, Condition=@n1.absoluteValue, False=@Remap.Result) {template="DISPATCH_RigVMDispatch_If(in Condition,in True,in False,out Result)"}.Result).Result
    B=@n1.premult
   )
   > RigVMFunctionReturnNode(outputQuat=@n2.Result, outputWeight=@Multiply.Result)
  ]}
 New Function_ContainedGraph {nodes [
   RigVMFunctionEntryNode()
   > DeltaRotation v4 = RigVMFunctionReferenceNode(
    Rotation Order=YXZ
    child=(Type=Bone,Name="robot_upperarm_l")
    outputMinMax=(X=0,Y=1)
    parent=(Type=Null,Name="upperarm_l_null")
    premult=1
    rotationMinMax=(X=0,Y=-90)
    triggerAxis=Y
    twistAxis=(X=1,Y=0,Z=0)
   )
   > &RigUnit_ModifyTransforms(
    ItemToModify=((Item=(Type=Bone,Name="base_shoulderpad_l"),Transform=(Rotation=(X=-0.0799,Y=-0.1331,Z=-0.0868,W=-0.9841),Translation=(X=10.4865,Y=-3.3643,Z=0.9154),Scale3D=(X=1,Y=1,Z=1))))
    WeightMinimum=0
    Weight=@"DeltaRotation v4".outputWeight
   )
   > DeltaRotation v4_1 = RigVMFunctionReferenceNode(
    Rotation Order=YXZ
    child=(Type=Bone,Name="robot_upperarm_l")
    outputMinMax=(X=0,Y=1)
    parent=(Type=Null,Name="upperarm_l_null")
    premult=1
    rotationMinMax=(X=0,Y=-70)
    triggerAxis=Z
    twistAxis=(X=1,Y=0,Z=0)
   )
   > &RigUnit_ModifyTransforms(
    ItemToModify=((Item=(Type=Bone,Name="base_shoulderpad_l"),Transform=(Rotation=(X=0.3647,Y=-0.045,Z=-0.0583,W=-0.9282),Translation=(X=3.5931,Y=4.6071,Z=-1.0896),Scale3D=(X=1,Y=1,Z=1))))
    WeightMinimum=0
    Weight=@"DeltaRotation v4_1".outputWeight
   )
   > DeltaRotation v4_2 = RigVMFunctionReferenceNode(
    Rotation Order=YXZ
    child=(Type=Bone,Name="robot_upperarm_l")
    outputMinMax=(X=0,Y=1)
    parent=(Type=Null,Name="upperarm_l_null")
    premult=1
    rotationMinMax=(X=0,Y=90)
    triggerAxis=Z
    twistAxis=(X=1,Y=0,Z=0)
   )
   > &RigUnit_ModifyTransforms(
    ItemToModify=((Item=(Type=Bone,Name="base_shoulderpad_l"),Transform=(Rotation=(X=-0.408,Y=-0.0423,Z=0.0941,W=-0.9071),Translation=(X=-0.8034,Y=-4.4172,Z=-2.0719),Scale3D=(X=1,Y=1,Z=1))))
    WeightMinimum=0
    Weight=@"DeltaRotation v4_2".outputWeight
   )
   > DeltaRotation v4_3 = RigVMFunctionReferenceNode(
    Rotation Order=YXZ
    child=(Type=Bone,Name="robot_upperarm_r")
    outputMinMax=(X=0,Y=1)
    parent=(Type=Null,Name="upperarm_r_null")
    premult=1
    rotationMinMax=(X=0,Y=-90)
    triggerAxis=Y
    twistAxis=(X=1,Y=0,Z=0)
   )
   > &RigUnit_ModifyTransforms(
    ItemToModify="((Item=(Type=Bone,Name=\"base_shoulderpad_r\"),Transform=(Rotation=(X=-0.0678,Y=-0.0721,Z=0.0049,W=-0.9951),Translation=(X=-12.9276,Y=0,Z=0.7134),Scale3D=(X=1,Y=1,Z=1))),(Item=(Type=Bone,Name=\"MSTR_handle\"),Transform=(Rotation=(X=0,Y=0,Z=-0.0624,W=-0.9981),T ...(truncated)"
    WeightMinimum=0
    Weight=@"DeltaRotation v4_3".outputWeight
   )
   > DeltaRotation v4_4 = RigVMFunctionReferenceNode(
    Rotation Order=YXZ
    child=(Type=Bone,Name="robot_upperarm_r")
    outputMinMax=(X=0,Y=1)
    parent=(Type=Null,Name="upperarm_r_null")
    premult=1
    rotationMinMax=(X=0,Y=-70)
    triggerAxis=Z
    twistAxis=(X=1,Y=0,Z=0)
   )
   > &RigUnit_ModifyTransforms(
    ItemToModify=((Item=(Type=Bone,Name="base_shoulderpad_r"),Transform=(Rotation=(X=0.406,Y=-0.1328,Z=-0.0311,W=-0.9036),Translation=(X=-4.1607,Y=-3.7211,Z=0.527),Scale3D=(X=1,Y=1,Z=1))))
    WeightMinimum=0
    Weight=@"DeltaRotation v4_4".outputWeight
   )
   > DeltaRotation v4_5 = RigVMFunctionReferenceNode(
    Rotation Order=YXZ
    child=(Type=Bone,Name="robot_upperarm_r")
    outputMinMax=(X=0,Y=1)
    parent=(Type=Null,Name="upperarm_r_null")
    premult=1
    rotationMinMax=(X=10,Y=90)
    triggerAxis=Z
    twistAxis=(X=1,Y=0,Z=0)
   )
   > &RigUnit_ModifyTransforms(
    ItemToModify="((Item=(Type=Bone,Name=\"base_shoulderpad_r\"),Transform=(Rotation=(X=0.1461,Y=0.1205,Z=-0.0032,W=0.9819),Translation=(X=-2.5037,Y=0.8938,Z=-1.3674),Scale3D=(X=1,Y=1,Z=1))),(Item=(Type=Bone,Name=\"MSTR_handle\"),Transform=(Rotation=(X=0,Y=0,Z=0.1224,W=0.9925), ...(truncated)"
    WeightMinimum=0
    Weight=@"DeltaRotation v4_5".outputWeight
   )
   RigVMFunction_MathBoolToFloat = &RigVMFunction_MathBoolToFloat(Value=RigVMVariableNode() {variable=targeting}.Value)
   > &RigUnit_ModifyTransforms(
    ItemToModify=((Item=(Type=Bone,Name="shoulderpad_r"),Transform=(Rotation=(X=0.5,Y=0,Z=0,W=0.866),Translation=(X=3,Y=6,Z=2),Scale3D=(X=1,Y=1,Z=1))))
    WeightMinimum=0
    Weight=/Script/RigVM.RigVMFunction_AnimEasing(
     SourceMaximum=1
     SourceMinimum=0
     TargetMaximum=1
     TargetMinimum=0
     Type=SineEaseOut
     Value=&RigVMFunction_KalmanFloat(BufferSize=10, Value=@RigVMFunction_MathBoolToFloat.Result) {template="KalmanFilter::Execute(in Value,in BufferSize,out Result)"}.Result
    ) {is=struct RigVMUnitNode}.Result
   )
   > RigVMFunctionReturnNode()
  ]}
 New Function_ContainedGraph {nodes [
   RigVMFunctionEntryNode()
   > &RigUnit_ParentConstraint(
    AdvancedSettings=(InterpolationType=Average,RotationOrderForFilter=XZY)
    Child=(Type=Bone,Name="MSTR_arm_strap_02_r")
    Parents=((Item=(Type=Bone,Name="deform_wrist_up_r"),Weight=0.15),(Item=(Type=Bone,Name="lowerarm_twist_01_r"),Weight=0.85))
    Weight=1
   )
   > &RigUnit_ParentConstraint(
    AdvancedSettings=(InterpolationType=Average,RotationOrderForFilter=XZY)
    Child=(Type=Bone,Name="MSTR_handle")
    Parents=((Item=(Type=Bone,Name="robot_upperarm_r"),Weight=0.11),(Item=(Type=Bone,Name="deform_elbow_out_r"),Weight=0.33),(Item=(Type=Bone,Name="lowerarm_twist_02_r"),Weight=0.29),(Item=(Type=Bone,Name="lowerarm_r"),Weight=0.27))
    Weight=1
   )
   > &RigUnit_ParentConstraint(
    AdvancedSettings=(InterpolationType=Average,RotationOrderForFilter=XZY)
    Child=(Type=Bone,Name="robot_upperarm_l")
    Filter=(TranslationFilter=(bX=true,bY=true,bZ=true),RotationFilter=(bX=false,bY=false,bZ=false),ScaleFilter=(bX=false,bY=false,bZ=false))
    Parents=((Item=(Type=Bone,Name="spine_05"),Weight=0.75),(Item=(Type=Bone,Name="clavicle_l"),Weight=0.25))
    Weight=1
   )
   > &RigUnit_ParentConstraint(
    AdvancedSettings=(InterpolationType=Average,RotationOrderForFilter=XZY)
    Child=(Type=Bone,Name="robot_upperarm_r")
    Filter=(TranslationFilter=(bX=true,bY=true,bZ=true),RotationFilter=(bX=false,bY=false,bZ=false),ScaleFilter=(bX=false,bY=false,bZ=false))
    Parents=((Item=(Type=Bone,Name="spine_05"),Weight=0.75),(Item=(Type=Bone,Name="clavicle_r"),Weight=0.25))
    Weight=1
   )
   > DeltaRotation v4_2 = RigVMFunctionReferenceNode(
    Rotation Order=YXZ
    child=(Type=Bone,Name="robot_upperarm_l")
    outputMinMax=(X=0,Y=1)
    parent=(Type=Null,Name="upperarm_l_null")
    premult=1
    rotationMinMax=(X=10,Y=90)
    triggerAxis=Z
    twistAxis=(X=1,Y=0,Z=0)
   )
   > &RigUnit_ModifyTransforms(
    ItemToModify="((Item=(Type=Bone,Name=\"robot_upperarm_l\"),Transform=(Rotation=(X=0,Y=0,Z=0,W=1),Translation=(X=2.762,Y=-3.7528,Z=2.1061),Scale3D=(X=1,Y=1,Z=1))),(Item=(Type=Bone,Name=\"robot_arm_transition_l\"),Transform=(Rotation=(X=0,Y=0,Z=0,W=1),Translation=(X=0,Y=0,Z=0 ...(truncated)"
    WeightMinimum=0
    Weight=@"DeltaRotation v4_2".outputWeight
   )
   > DeltaRotation v4_4 = RigVMFunctionReferenceNode(
    Rotation Order=YXZ
    child=(Type=Bone,Name="robot_upperarm_l")
    outputMinMax=(X=0,Y=1)
    parent=(Type=Null,Name="upperarm_l_null")
    premult=1
    rotationMinMax=(X=0,Y=-90)
    triggerAxis=Y
    twistAxis=(X=1,Y=0,Z=0)
   )
   > &RigUnit_ModifyTransforms(
    ItemToModify=((Item=(Type=Bone,Name="robot_upperarm_l"),Transform=(Rotation=(X=0,Y=0,Z=0,W=1),Translation=(X=3,Y=0,Z=-1),Scale3D=(X=1,Y=1,Z=1))))
    WeightMinimum=0
    Weight=@"DeltaRotation v4_4".outputWeight
   )
   > DeltaRotation v4_1 = RigVMFunctionReferenceNode(
    Rotation Order=YXZ
    child=(Type=Bone,Name="robot_upperarm_r")
    outputMinMax=(X=0,Y=1)
    parent=(Type=Null,Name="upperarm_r_null")
    premult=1
    rotationMinMax=(X=10,Y=90)
    triggerAxis=Z
    twistAxis=(X=1,Y=0,Z=0)
   )
   > &RigUnit_ModifyTransforms(
    ItemToModify=((Item=(Type=Bone,Name="robot_upperarm_r"),Transform=(Rotation=(X=0,Y=0,Z=0,W=1),Translation=(X=-2.762,Y=3.7528,Z=-2.1061),Scale3D=(X=1,Y=1,Z=1))))
    WeightMinimum=0
    Weight=@"DeltaRotation v4_1".outputWeight
   )
   > DeltaRotation v4_3 = RigVMFunctionReferenceNode(
    Rotation Order=YXZ
    child=(Type=Bone,Name="robot_upperarm_r")
    outputMinMax=(X=0,Y=1)
    parent=(Type=Null,Name="upperarm_r_null")
    premult=1
    rotationMinMax=(X=0,Y=-90)
    triggerAxis=Y
    twistAxis=(X=1,Y=0,Z=0)
   )
   > &RigUnit_ModifyTransforms(
    ItemToModify=((Item=(Type=Bone,Name="robot_upperarm_r"),Transform=(Rotation=(X=0,Y=0,Z=0,W=1),Translation=(X=-3.5,Y=0,Z=2.5),Scale3D=(X=1,Y=1,Z=1))))
    WeightMinimum=0
    Weight=@"DeltaRotation v4_3".outputWeight
   )
   > &RigUnit_AimConstraintLocalSpaceOffset(
    AdvancedSettings=(DebugSettings=(bEnabled=false,Scale=100,WorldOffset=(Rotation=(X=0,Y=0,Z=0,W=1),Translation=(X=0,Y=0,Z=0),Scale3D=(X=1,Y=1,Z=1))),RotationOrderForFilter=XZY)
    Child=(Type=Bone,Name="robot_upperarm_l")
    Filter=(bX=true,bY=true,bZ=true)
    Parents=((Item=(Type=Bone,Name="lowerarm_l"),Weight=1))
    Weight=1
    WorldUp=(Target=(X=0,Y=0,Z=1),Kind=Direction,Space=(Type=Bone,Name="upperarm_l"))
   )
   > &RigUnit_AimConstraintLocalSpaceOffset(
    AdvancedSettings=(DebugSettings=(bEnabled=false,Scale=100,WorldOffset=(Rotation=(X=0,Y=0,Z=0,W=1),Translation=(X=0,Y=0,Z=0),Scale3D=(X=1,Y=1,Z=1))),RotationOrderForFilter=XZY)
    AimAxis=(X=-1,Y=0,Z=0)
    Child=(Type=Bone,Name="robot_upperarm_r")
    Filter=(bX=true,bY=true,bZ=true)
    Parents=((Item=(Type=Bone,Name="lowerarm_r"),Weight=1))
    Weight=1
    WorldUp=(Target=(X=0,Y=0,Z=-1),Kind=Direction,Space=(Type=Bone,Name="upperarm_r"))
   )
   > /Script/ControlRig.RigUnit_PositionConstraintLocalSpaceOffset(
    Child=(Type=Bone,Name="robot_upperarmEnd_l")
    Filter=(bX=True,bY=True,bZ=True)
    Parents=((Item=(Type=Bone,Name="elbow_sphere_l"),Weight=1))
    Weight=1
    bMaintainOffset=True
   ) {is=struct RigVMUnitNode}
   > &RigUnit_ParentConstraint(
    AdvancedSettings=(InterpolationType=Average,RotationOrderForFilter=XZY)
    Child=(Type=Bone,Name="robot_arm_transition_l")
    Parents=((Item=(Type=Bone,Name="spine_05"),Weight=0.5),(Item=(Type=Bone,Name="robot_upperarm_l"),Weight=0.5))
    Weight=1
   )
   > &RigUnit_ParentConstraint(
    AdvancedSettings=(InterpolationType=Average,RotationOrderForFilter=XZY)
    Child=(Type=Bone,Name="robot_arm_transition_r")
    Parents=((Item=(Type=Bone,Name="spine_05"),Weight=0.5),(Item=(Type=Bone,Name="robot_upperarm_r"),Weight=0.5))
    Weight=1
   )
   > &RigUnit_ParentConstraint(
    AdvancedSettings=(InterpolationType=Average,RotationOrderForFilter=XZY)
    Child=(Type=Bone,Name="elbow_sphere_l")
    Filter=(TranslationFilter=(bX=false,bY=false,bZ=false),RotationFilter=(bX=True,bY=True,bZ=True),ScaleFilter=(bX=True,bY=True,bZ=True))
    Parents=((Item=(Type=Bone,Name="lowerarm_l"),Weight=0.5),(Item=(Type=Bone,Name="upperarm_l"),Weight=0.5))
    Weight=1
   )
   > &RigUnit_ParentConstraint(
    AdvancedSettings=(InterpolationType=Average,RotationOrderForFilter=XZY)
    Child=(Type=Bone,Name="elbow_sphere_l")
    Filter=(TranslationFilter=(bX=true,bY=true,bZ=true),RotationFilter=(bX=false,bY=false,bZ=false),ScaleFilter=(bX=True,bY=True,bZ=True))
    Parents=((Item=(Type=Bone,Name="lowerarm_l"),Weight=0.25),(Item=(Type=Bone,Name="upperarm_l"),Weight=0.75))
    Weight=1
   )
   > &RigUnit_PositionConstraintLocalSpaceOffset(
    Child=(Type=Bone,Name="robot_lowerarm_l")
    Filter=(bX=True,bY=True,bZ=True)
    Parents=((Item=(Type=Bone,Name="elbow_sphere_l"),Weight=1))
    Weight=1
    bMaintainOffset=True
   )
   > &RigUnit_AimConstraintLocalSpaceOffset(
    Child=(Type=Bone,Name="robot_lowerarm_l")
    Filter=(bX=True,bY=True,bZ=True)
    Parents=((Item=(Type=Bone,Name="hand_l"),Weight=1))
    Weight=1
    WorldUp=(Target=(X=0,Y=0,Z=1),Kind=Direction,Space=(Type=Bone,Name="lowerarm_l"))
   )
   > RigVMFunctionReturnNode()
  ]}
 New Function_ContainedGraph {nodes [
   RigVMFunctionEntryNode()
   > &RigUnit_ParentConstraint(
    AdvancedSettings=(InterpolationType=Average,RotationOrderForFilter=XZY)
    Child=(Type=Bone,Name="base_kneepad_l")
    Parents=((Item=(Type=Bone,Name="thigh_l"),Weight=0.75),(Item=(Type=Bone,Name="calf_l"),Weight=0.25))
    Weight=1
   )
   > &RigUnit_ParentConstraint(
    AdvancedSettings=(InterpolationType=Average,RotationOrderForFilter=XZY)
    Child=(Type=Bone,Name="MSTR_boot_strap_l")
    Parents=((Item=(Type=Bone,Name="foot_l"),Weight=0.5),(Item=(Type=Bone,Name="calf_l"),Weight=0.5))
    Weight=1
   )
   > DeltaRotation v4 = RigVMFunctionReferenceNode(
    Rotation Order=ZYX
    child=(Type=Bone,Name="calf_l")
    debugRotation=false
    outputMinMax=(X=0,Y=1)
    parent=(Type=Null,Name="knee_l_null")
    premult=1
    rotationMinMax=(X=90,Y=120)
    triggerAxis=Z
    twistAxis=(X=1,Y=0,Z=0)
   )
   > &RigUnit_ModifyTransforms(
    ItemToModify=((Item=(Type=Bone,Name="MSTR_boot_tassel_l"),Transform=(Rotation=(X=0,Y=0,Z=-0.9063,W=0.4226),Translation=(X=0,Y=0,Z=0),Scale3D=(X=1,Y=1,Z=1))))
    WeightMinimum=0
    Weight=@"DeltaRotation v4".outputWeight
   )
   > RigVMFunctionReturnNode()
  ]}
 New Function_ContainedGraph {nodes [
   RigVMFunctionEntryNode()
   > &RigUnit_ModifyTransforms(
    ItemToModify=((Item=(Type=Bone,Name="cardboard_back"),Transform=(Rotation=(X=0,Y=0,Z=0,W=1),Translation=(X=0,Y=-1.5,Z=0),Scale3D=(X=1,Y=1,Z=1))))
    Weight=1
    WeightMinimum=0
   )
   > &RigUnit_ParentConstraint(
    AdvancedSettings=(InterpolationType=Average,RotationOrderForFilter=XZY)
    Child=(Type=Bone,Name="power_strip")
    Parents=((Item=(Type=Bone,Name="spine_02"),Weight=0.32),(Item=(Type=Bone,Name="spine_01"),Weight=0.2),(Item=(Type=Bone,Name="pelvis"),Weight=0.15),(Item=(Type=Bone,Name="spine_05"),Weight=0.15),(Item=(Type=Bone,Name="spine_03"),Weight=0.18))
    Weight=1
   )
   > DeltaRotation v4 = RigVMFunctionReferenceNode(
    Rotation Order=ZYX
    child=(Type=Bone,Name="spine_01")
    debugRotation=false
    outputMinMax=(X=0,Y=1)
    parent=(Type=Bone,Name="pelvis")
    premult=1
    rotationMinMax=(X=-8,Y=-30)
    triggerAxis=Z
    twistAxis=(X=1,Y=0,Z=0)
   )
   > &RigUnit_ModifyTransforms(
    ItemToModify=((Item=(Type=Bone,Name="MSTR_pliers"),Transform=(Rotation=(X=-0.0678,Y=-0.1653,Z=-0.015,W=0.9838),Translation=(X=-0.9586,Y=0,Z=2.3672),Scale3D=(X=1,Y=1,Z=1))))
    WeightMinimum=0
    Weight=@"DeltaRotation v4".outputWeight
   )
   > &RigUnit_ParentConstraint(
    AdvancedSettings=(InterpolationType=Average,RotationOrderForFilter=XZY)
    Child=(Type=Null,Name="chest_tube_null_02_reverse")
    Parents=((Item=(Type=Bone,Name="dyn_pouch_01_l"),Weight=1))
    Weight=1
   )
   > &RigUnit_ParentConstraint(
    AdvancedSettings=(InterpolationType=Average,RotationOrderForFilter=XZY)
    Child=(Type=Null,Name="chest_tube_null_03")
    Filter=(TranslationFilter=(bX=True,bY=True,bZ=True),RotationFilter=(bX=true,bY=true,bZ=true),ScaleFilter=(bX=True,bY=True,bZ=True))
    Parents=((Item=(Type=Bone,Name="dyn_pouch_01_l"),Weight=0.25),(Item=(Type=Bone,Name="pelvis"),Weight=0.75))
    Weight=1
   )
   > &RigUnit_ParentConstraint(
    AdvancedSettings=(InterpolationType=Average,RotationOrderForFilter=XZY)
    Child=(Type=Null,Name="chest_tube_null_01_reverse")
    Parents=((Item=(Type=Bone,Name="dyn_pouch_01_l"),Weight=0.25),(Item=(Type=Bone,Name="pelvis"),Weight=0.75))
    Weight=1
   )
   > SplineFromItems_3 = RigVMFunctionReferenceNode(
    Items=((Type=Null,Name="chest_tube_null_01"),(Type=Null,Name="chest_tube_null_02"),(Type=Null,Name="chest_tube_null_03"),(Type=Null,Name="chest_tube_null_02_reverse"))
    Samples Per Segment=25
    Spline Mode=Hermite
   )
   > /Script/ControlRigSpline.RigUnit_DrawControlRigSpline(Color=(R=0.0433,G=1,B=0,A=1), Detail=16, Thickness=0.1, Spline=@SplineFromItems_3.Spline) {is=struct RigVMUnitNode}
   ItemArray_2 = RigVMDispatchNode(Value=((Type=Bone,Name="dyn_chest_tube_01"),(Type=Bone,Name="dyn_chest_tube_02"),(Type=Bone,Name="dyn_chest_tube_03"),(Type=Bone,Name="dyn_chest_tube_04"))) {template="DISPATCH_RigVMDispatch_Constant(io Value)"}
   > /Script/ControlRigSpline.RigUnit_FitChainToSplineCurveItemArray(
    Alignment=Front
    DebugSettings=(bEnabled=False,Scale=1,CurveColor=(R=1,G=1,B=0,A=1),SegmentsColor=(R=1,G=0,B=0,A=1),WorldOffset=(Rotation=(X=0,Y=0,Z=0,W=1),Translation=(X=0,Y=0,Z=0),Scale3D=(X=1,Y=1,Z=1)))
    Maximum=1
    Minimum=0
    PoleVectorPosition=(X=0,Y=0,Z=0)
    PrimaryAxis=(X=1,Y=0,Z=0)
    RotationEaseType=Linear
    Rotations=()
    SamplingPrecision=25
    SecondaryAxis=(X=0,Y=0,Z=0)
    Weight=1
    bPropagateToChildren=true
    Items=@ItemArray_2.Value
    Spline=@SplineFromItems_3.Spline
   ) {is=struct RigVMUnitNode}
   > SplineFromItems_4 = RigVMFunctionReferenceNode(
    Items=((Type=Null,Name="chest_tube_null_01_reverse"),(Type=Null,Name="chest_tube_null_02_reverse"),(Type=Null,Name="chest_tube_null_03"),(Type=Null,Name="chest_tube_null_02"))
    Samples Per Segment=25
    Spline Mode=Hermite
   )
   > &RigUnit_DrawControlRigSpline(Color=(R=1,G=0.2618,B=0.9308,A=1), Detail=16, Thickness=0.1, Spline=@SplineFromItems_4.Spline)
   ItemArray_3 = RigVMDispatchNode(Value=((Type=Bone,Name="dyn_chest_tube_01_reverse"),(Type=Bone,Name="dyn_chest_tube_02_reverse"),(Type=Bone,Name="dyn_chest_tube_03_reverse"))) {template="DISPATCH_RigVMDispatch_Constant(io Value)"}
   > &RigUnit_FitChainToSplineCurveItemArray(
    Alignment=Front
    DebugSettings=(bEnabled=False,Scale=1,CurveColor=(R=1,G=1,B=0,A=1),SegmentsColor=(R=1,G=0,B=0,A=1),WorldOffset=(Rotation=(X=0,Y=0,Z=0,W=1),Translation=(X=0,Y=0,Z=0),Scale3D=(X=1,Y=1,Z=1)))
    Maximum=1
    Minimum=0
    PoleVectorPosition=(X=0,Y=0,Z=0)
    PrimaryAxis=(X=1,Y=0,Z=0)
    RotationEaseType=Linear
    Rotations=()
    SamplingPrecision=25
    SecondaryAxis=(X=0,Y=0,Z=0)
    Weight=1
    bPropagateToChildren=true
    Items=@ItemArray_3.Value
    Spline=@SplineFromItems_4.Spline
   )
   > &RigUnit_ParentConstraint(
    AdvancedSettings=(InterpolationType=Average,RotationOrderForFilter=XZY)
    Child=(Type=Bone,Name="dyn_chest_tube_02_reverse")
    Filter=(TranslationFilter=(bX=True,bY=True,bZ=True),RotationFilter=(bX=false,bY=false,bZ=false),ScaleFilter=(bX=True,bY=True,bZ=True))
    Parents=((Item=(Type=Bone,Name="dyn_screwdriver"),Weight=1))
    Weight=1
   )
   > DeltaRotation v4_5 = RigVMFunctionReferenceNode(
    Rotation Order=ZYX
    child=(Type=Bone,Name="thigh_l")
    debugRotation=false
    outputMinMax=(X=0,Y=1)
    parent=(Type=Null,Name="thigh_l_null")
    premult=1
    rotationMinMax=(X=-10,Y=-90)
    triggerAxis=Z
    twistAxis=(X=1,Y=0,Z=0)
   )
   > &RigUnit_ModifyTransforms(
    ItemToModify="((Item=(Type=Bone,Name=\"dyn_chest_tube_02_reverse\"),Transform=(Rotation=(X=0,Y=0,Z=0,W=1),Translation=(X=0,Y=0,Z=-6),Scale3D=(X=1,Y=1,Z=1))),(Item=(Type=Bone,Name=\"dyn_chest_tube_01_reverse\"),Transform=(Rotation=(X=0,Y=0.0872,Z=0,W=0.9962),Translation=(X=0 ...(truncated)"
    WeightMinimum=0
    Weight=@"DeltaRotation v4_5".outputWeight
   )
   > DeltaRotation v4_6 = RigVMFunctionReferenceNode(
    Rotation Order=ZYX
    child=(Type=Bone,Name="thigh_l")
    debugRotation=false
    outputMinMax=(X=0,Y=1)
    parent=(Type=Null,Name="thigh_l_null")
    premult=1
    rotationMinMax=(X=0,Y=-90)
    triggerAxis=Y
    twistAxis=(X=1,Y=0,Z=0)
   )
   > &RigUnit_ModifyTransforms(
    ItemToModify=((Item=(Type=Bone,Name="dyn_chest_tube_02_reverse"),Transform=(Rotation=(X=-0.0044,Y=-0.0917,Z=0.0482,W=0.9946),Translation=(X=-0.81,Y=-0.8449,Z=1.9218),Scale3D=(X=1,Y=1,Z=1))))
    WeightMinimum=0
    Weight=@"DeltaRotation v4_6".outputWeight
   )
   > RigVMFunctionReturnNode()
  ]}
 RigVMFunctionLibrary {nodes [
   DeltaRotation v4 = RigVMCollapseNode(
    ExecuteContext=()
    Rotation Order=ZYX
    child=(Type=Bone,Name="None")
    outputMinMax=(X=0,Y=1)
    parent=(Type=Bone,Name="None")
    premult=1
    rotationMinMax=(X=0,Y=0)
    twistAxis=(X=1,Y=0,Z=0)
   )
   arms = RigVMCollapseNode(ExecuteContext=())
   legs = RigVMCollapseNode(ExecuteContext=())
   pouches = RigVMCollapseNode(ExecuteContext=())
   shoulderpads = RigVMCollapseNode(ExecuteContext=())
   spine = RigVMCollapseNode(ExecuteContext=())
  ]}
 RigVMModel {nodes [
   RigUnit_BeginExecution = &RigUnit_BeginExecution()
   FN_DeformRig_FNC = RigVMFunctionReferenceNode(
    DeformRig_ Curve="(deformRig_l_upperarm_up_alpha_72_#1=true,deformRig_r_upperarm_up_alpha_73_#2=true,deformRig_l_upperarm_fr_alpha_74_#3=true,deformRig_r_upperarm_fr_alpha_75_CB51B80C4 ...(truncated)"
    DeformRig_Bool="(upperarm_1_#4=true,elbow_3_#5=True,wrist_7_#6=True,groin_8_#7=True,glute_17_#8=true,knee_9_40C408324CA261 ...(truncated)"
    DeformRig_Debug=(upperarm_debug_18_#4=False,wrist_debug_20_#6=False,lat_debug_26_#9=False,groinGlute_debug_27_#7=False)
    DeformRig_Value="(upperarm_up_18_#10=(Rotation=(X=0.0,Y=-0.0,Z=0.0,W=1.0),Translation=(X=0.0,Y=0.0,Z=0),Scale3D=(X=1,Y=1.0,Z=1.0)),upperarm_fwd_fr_19_#11=(Rotation=(X=0.0,Y=-0.0,Z=0.0,W=1.0),Translation=(X=0.0,Y=0,Z ...(truncated)"
    ExecuteContext=@RigUnit_BeginExecution.ExecutePin
   )
   > legs = RigVMFunctionReferenceNode()
   > pouches = RigVMFunctionReferenceNode()
   > arms = RigVMFunctionReferenceNode()
   n1 = &RigVMFunction_ControlFlowBranch(Condition=RigVMVariableNode() {variable=FE idle}.Value, ExecuteContext=@arms.ExecuteContext)
   spine = RigVMFunctionReferenceNode(ExecuteContext=@n1.Completed)
   shoulderpads = RigVMFunctionReferenceNode(ExecuteContext=@n1.False)
  ]}
 RigVMModel Construction Event Graph {nodes [
   n1 = &RigUnit_PrepareForExecution()
   &RigUnit_HierarchyAddNull(
    Name=knee_l_null
    Parent=(Type=Bone,Name="thigh_l")
    Space=GlobalSpace
    Transform=&RigUnit_GetTransform(Item=(Type=Bone,Name="calf_l"), bInitial=false).Transform
    ExecutePin=@n1.ExecutePin
   )
   > &RigUnit_HierarchyAddNull(
    Name=thigh_l_null
    Parent=(Type=Bone,Name="pelvis")
    Space=GlobalSpace
    Transform=&RigUnit_GetTransform(Item=(Type=Bone,Name="thigh_l"), bInitial=false).Transform
   )
   > &RigUnit_HierarchyAddNull(
    Name=thigh_r_null
    Parent=(Type=Bone,Name="pelvis")
    Space=GlobalSpace
    Transform=&RigUnit_GetTransform(Item=(Type=Bone,Name="thigh_r"), bInitial=false).Transform
   )
   > &RigUnit_HierarchyAddNull(
    Name=upperarm_l_null
    Parent=(Type=Bone,Name="spine_05")
    Space=GlobalSpace
    Transform=&RigUnit_GetTransform(Item=(Type=Bone,Name="upperarm_l"), bInitial=false).Transform
   )
   > &RigUnit_HierarchyAddNull(
    Name=upperarm_r_null
    Parent=(Type=Bone,Name="spine_05")
    Space=GlobalSpace
    Transform=&RigUnit_GetTransform(Item=(Type=Bone,Name="upperarm_r"), bInitial=false).Transform
   )
   ItemArray = RigVMDispatchNode(Value=((Type=Bone,Name="dyn_chest_tube_01"),(Type=Bone,Name="dyn_chest_tube_02"),(Type=Bone,Name="dyn_chest_tube_03"),(Type=Bone,Name="dyn_chest_tube_02_reverse"),(Type=Bone,Name="dyn_chest_tube_01_reverse"))) {template="DISPATCH_RigVMDispatch_Constant(io Value)"}
   > For_Each_1 = RigVMDispatchNode(Array=/Script/ControlRig.RigUnit_GetTransformItemArray(Space=GlobalSpace, bInitial=False, Items=@ItemArray.Value) {is=struct RigVMUnitNode}.Transforms) {template="DISPATCH_RigVMDispatch_ArrayIterator(in Array,out Element,out Index,out Count,out Ratio)"}
   n2 = RigVMDispatchNode(Index=@For_Each_1.Index, Array=@ItemArray.Value) {template="DISPATCH_RigVMDispatch_ArrayGetAtIndex(in Array,in Index,out Element)"}
   > &RigUnit_HierarchyAddNull(
    Parent=(Type=Bone,Name="spine_02")
    Space=GlobalSpace
    Transform=@For_Each_1.Element
    Name=/Script/RigVM.RigVMFunction_NameReplace(
     Old=dyn_
     Name=&RigVMFunction_NameReplace(New=tube_null, Old=tube, Name=@n2.Element.Name) {template="Replace::Execute(in Name,in Old,in New,out Result)"}.Result
    ) {is=struct RigVMUnitNode, template="Replace::Execute(in Name,in Old,in New,out Result)"}.Result
   )
   /Script/ControlRig.RigUnit_SwitchParent(
    Child=(Type=Null,Name="chest_tube_null_02_reverse")
    Mode=ParentItem
    Parent=(Type=Bone,Name="dyn_pouch_01_l")
    bMaintainGlobal=True
    ExecutePin=@For_Each_1.Completed
   ) {is=struct RigVMUnitNode}
   > &RigUnit_SwitchParent(Child=(Type=Null,Name="chest_tube_null_02"), Mode=ParentItem, Parent=(Type=Bone,Name="spine_03"), bMaintainGlobal=True)
   > &RigUnit_SwitchParent(Child=(Type=Null,Name="chest_tube_null_01"), Mode=ParentItem, Parent=(Type=Bone,Name="spine_05"), bMaintainGlobal=True)
  ]}
]
hierarchy [
 l_arm_bk {type=Curve, loc=<0 0 0>, rot=<0 0 0 1>}
 l_arm_ft {type=Curve, loc=<0 0 0>, rot=<0 0 0 1>}
 l_arm_up {type=Curve, loc=<0 0 0>, rot=<0 0 0 1>}
 r_arm_bk {type=Curve, loc=<0 0 0>, rot=<0 0 0 1>}
 r_arm_ft {type=Curve, loc=<0 0 0>, rot=<0 0 0 1>}
 r_arm_up {type=Curve, loc=<0 0 0>, rot=<0 0 0 1>}
 root {loc=<0 0 0>, rot=<0 0 0 1>}
 pelvis {parent=root, loc=<0 -0.7321 92.4218>, rot=<0.0127 0.707 -0.0127 -0.707>}
 spine_01 {parent=pelvis, loc=<5.6002 -0.5128 0>, rot=<0 0 0.034 -0.9994>}
 spine_02 {parent=spine_01, loc=<8.0756 0.0139 0>, rot=<0 0 0.0238 -0.9997>}
 spine_03 {parent=spine_02, loc=<7.0395 0 0>, rot=<0 0 0.0149 -0.9999>}
 spine_04 {parent=spine_03, loc=<8.1023 0 0>, rot=<0 0 -0.0132 -0.9999>}
 spine_05 {parent=spine_04, loc=<10.0141 0 0>, rot=<0 0 -0.0015 -1>}
 clavicle_l {parent=spine_05, loc=<2.5856 -0.2866 -4.3777>, rot=<0.0465 -0.7289 -0.009 -0.6829>}
 upperarm_l {parent=clavicle_l, loc=<18.2264 0.1342 -0.001>, rot=<0.0253 -0.2693 0.0013 -0.9627>}
 lowerarm_l {parent=upperarm_l, loc=<29.0071 0.1686 -1.239>, rot=<0.0391 -0.0099 -0.2453 -0.9686>}
 hand_l {parent=lowerarm_l, loc=<25.8607 0.222 1.0274>, rot=<-0.7193 -0.0988 0.1193 -0.6773>}
 thumb_01_l {parent=hand_l, loc=<1.3118 -0.3191 -2.459>, rot=<0.6546 -0.4042 0.0664 -0.6354>}
 thumb_02_l {parent=thumb_01_l, loc=<5.9336 -0.0448 0.0401>, rot=<-0.0037 0.0213 0.298 -0.9543>}
 thumb_03_l {parent=thumb_02_l, loc=<2.2223 0.3682 -0.0907>, rot=<0.0118 -0.013 0.0066 -0.9998>}
 index_metacarpal_l {parent=hand_l, loc=<1.6535 0.0183 -2.96>, rot=<0.0158 -0.0124 -0.0173 -0.9996>}
 index_01_l {parent=index_metacarpal_l, loc=<8.3008 0.0357 -0.0524>, rot=<-0.0044 0.0485 0.0924 -0.9945>}
 index_02_l {parent=index_01_l, loc=<3.3587 -0.3004 0.0452>, rot=<-0.002 0.0027 0.3225 -0.9466>}
 index_03_l {parent=index_02_l, loc=<2.5578 -0.123 -0.0252>, rot=<-0.002 0.0067 0.2558 -0.9667>}
 middle_metacarpal_l {parent=hand_l, loc=<2.1856 -0.1725 -0.5984>, rot=<0.006 0.0168 -0.0329 -0.9993>}
 middle_01_l {parent=middle_metacarpal_l, loc=<7.3275 -0.2447 0.0129>, rot=<-0.0208 0.0042 0.0858 -0.9961>}
 middle_02_l {parent=middle_01_l, loc=<4.0123 0.1401 0.1175>, rot=<-0.0218 0.0258 0.3512 -0.9357>}
 middle_03_l {parent=middle_02_l, loc=<2.6077 0.0534 0.1373>, rot=<0.0082 0.0188 0.1708 -0.9851>}
 ring_metacarpal_l {parent=hand_l, loc=<2.2074 -0.3625 1.3638>, rot=<-0.0284 0.0722 -0.0458 -0.9959>}
 ring_01_l {parent=ring_metacarpal_l, loc=<7.0295 -0.6197 -0.1446>, rot=<-0.0489 -0.0252 0.1892 -0.9804>}
 ring_02_l {parent=ring_01_l, loc=<3.6174 0.4647 -0.1027>, rot=<0 0 0.2806 -0.9598>}
 ring_03_l {parent=ring_02_l, loc=<2.4152 0.114 0.0199>, rot=<0 0 0.1189 -0.9929>}
 pinky_metacarpal_l {parent=hand_l, loc=<1.7616 -0.9753 3.3814>, rot=<-0.0397 0.0814 -0.069 -0.9935>}
 pinky_01_l {parent=pinky_metacarpal_l, loc=<7.1709 -0.4831 -0.1094>, rot=<-0.0654 -0.0035 0.2242 -0.9723>}
 pinky_02_l {parent=pinky_01_l, loc=<2.8274 0.0893 -0.1043>, rot=<0.0037 0.0011 0.2854 -0.9584>}
 pinky_03_l {parent=pinky_02_l, loc=<2.2046 0.0307 -0.0078>, rot=<-0.0002 0.0098 0.1445 -0.9895>}
 deform_wrist_dn_base_l {parent=hand_l, loc=<0 -3.4173 0>, rot=<0 0 0 -1>}
 deform_wrist_dn_l {parent=deform_wrist_dn_base_l, loc=<0 0 0>, rot=<0 0 0 1>}
 deform_wrist_up_base_l {parent=hand_l, loc=<0 2.7805 0>, rot=<0 0 0 -1>}
 deform_wrist_up_l {parent=deform_wrist_up_base_l, loc=<0 0 0>, rot=<0 0 0 1>}
 lowerarm_twist_02_l {parent=lowerarm_l, loc=<9.6199 0 0>, rot=<0 0 0 -1>}
 lowerarm_twist_01_l {parent=lowerarm_l, loc=<17.8549 0 0>, rot=<0 0 0 -1>}
 elbow_sphere_l {parent=lowerarm_l, loc=<-2.1071 0.1869 0.1539>, rot=<0 0 0 1>}
 robot_lowerarm_l {parent=lowerarm_l, loc=<-2.1071 0.1869 0.1539>, rot=<-0.0473 0.0161 -0.001 -0.9987>}
 robot_upperarm_l {parent=upperarm_l, loc=<3.3767 0.2443 0.8522>, rot=<-0.0065 -0.0392 0.0192 -0.999>}
 base_shoulderpad_l {parent=robot_upperarm_l, loc=<1.2437 -0.126 4.9442>, rot=<0.0065 0.0392 -0.0192 -0.999>}
 shoulderpad_l {parent=base_shoulderpad_l, loc=<0 0 0>, rot=<0 0 0 1>}
 robot_upperarmEnd_l {parent=robot_upperarm_l, loc=<23.779 0 0>, rot=<0.0551 0.0273 -0.262 -0.9631>}
 robot_arm_transition_l {parent=robot_upperarm_l, loc=<0 0 0>, rot=<0 0 0 1>}
 neck_01 {parent=spine_05, loc=<10.815 1.9196 0>, rot=<0 0 -0.1864 -0.9825>}
 neck_02 {parent=neck_01, loc=<4.7347 0.3045 0>, rot=<0 0 0.0363 -0.9993>}
 head {parent=neck_02, loc=<5.7632 0.5552 0>, rot=<0 0 0.115 -0.9934>}
 clavicle_r {parent=spine_05, loc=<2.5855 -0.2866 4.3777>, rot=<0.7289 0.0465 -0.6829 0.009>}
 upperarm_r {parent=clavicle_r, loc=<-18.2264 -0.1342 0.001>, rot=<0.0253 -0.2693 0.0013 -0.9627>}
 lowerarm_r {parent=upperarm_r, loc=<-29.0071 -0.1686 1.239>, rot=<0.0391 -0.0099 -0.2453 -0.9686>}
 hand_r {parent=lowerarm_r, loc=<-25.8607 -0.222 -1.0274>, rot=<-0.7193 -0.0988 0.1193 -0.6773>}
 thumb_01_r {parent=hand_r, loc=<-1.3118 0.3191 2.459>, rot=<0.6546 -0.4042 0.0664 -0.6354>}
 thumb_02_r {parent=thumb_01_r, loc=<-5.9336 0.0448 -0.0401>, rot=<-0.0037 0.0213 0.298 -0.9543>}
 thumb_03_r {parent=thumb_02_r, loc=<-2.2223 -0.3682 0.0907>, rot=<0.0118 -0.013 0.0066 -0.9998>}
 index_metacarpal_r {parent=hand_r, loc=<-1.6535 -0.0183 2.96>, rot=<0.0158 -0.0124 -0.0173 -0.9996>}
 index_01_r {parent=index_metacarpal_r, loc=<-8.3008 -0.0357 0.0524>, rot=<-0.0044 0.0485 0.0924 -0.9945>}
 index_02_r {parent=index_01_r, loc=<-3.3587 0.3004 -0.0452>, rot=<-0.002 0.0027 0.3225 -0.9466>}
 index_03_r {parent=index_02_r, loc=<-2.5578 0.123 0.0252>, rot=<-0.002 0.0067 0.2558 -0.9667>}
 middle_metacarpal_r {parent=hand_r, loc=<-2.1856 0.1725 0.5984>, rot=<0.006 0.0168 -0.0329 -0.9993>}
 middle_01_r {parent=middle_metacarpal_r, loc=<-7.3275 0.2447 -0.0129>, rot=<-0.0208 0.0042 0.0858 -0.9961>}
 middle_02_r {parent=middle_01_r, loc=<-4.0123 -0.1401 -0.1175>, rot=<-0.0218 0.0258 0.3512 -0.9357>}
 middle_03_r {parent=middle_02_r, loc=<-2.6077 -0.0534 -0.1373>, rot=<0.0082 0.0188 0.1708 -0.9851>}
 ring_metacarpal_r {parent=hand_r, loc=<-2.2074 0.3625 -1.3638>, rot=<-0.0284 0.0722 -0.0458 -0.9959>}
 ring_01_r {parent=ring_metacarpal_r, loc=<-7.0295 0.6197 0.1446>, rot=<-0.0489 -0.0252 0.1892 -0.9804>}
 ring_02_r {parent=ring_01_r, loc=<-3.6174 -0.4647 0.1027>, rot=<0 0 0.2806 -0.9598>}
 ring_03_r {parent=ring_02_r, loc=<-2.4152 -0.114 -0.0199>, rot=<0 0 0.1189 -0.9929>}
 pinky_metacarpal_r {parent=hand_r, loc=<-1.7616 0.9753 -3.3814>, rot=<-0.0397 0.0814 -0.069 -0.9935>}
 pinky_01_r {parent=pinky_metacarpal_r, loc=<-7.1709 0.4831 0.1094>, rot=<-0.0654 -0.0035 0.2242 -0.9723>}
 pinky_02_r {parent=pinky_01_r, loc=<-2.8274 -0.0893 0.1043>, rot=<0.0037 0.0011 0.2854 -0.9584>}
 pinky_03_r {parent=pinky_02_r, loc=<-2.2046 -0.0307 0.0078>, rot=<-0.0002 0.0098 0.1445 -0.9895>}
 deform_wrist_up_base_r {parent=hand_r, loc=<-0.0001 -2.7804 0>, rot=<0 0 0 -1>}
 deform_wrist_up_r {parent=deform_wrist_up_base_r, loc=<0 0 0>, rot=<0 0 0 -1>}
 lowerarm_twist_02_r {parent=lowerarm_r, loc=<-9.6199 0 0>, rot=<0 0 0 -1>}
 lowerarm_twist_01_r {parent=lowerarm_r, loc=<-17.8549 0 0>, rot=<0 0 0 -1>}
 MSTR_arm_strap_02_r {parent=lowerarm_twist_01_r, loc=<-5.5991 -0.74 -4.7845>, rot=<-0.5748 0.2548 0.7147 -0.3064>}
 dyn_arm_strap_02_r {parent=MSTR_arm_strap_02_r, loc=<0 0 0>, rot=<0 0 0 1>}
 deform_elbow_in_base_r {parent=lowerarm_r, loc=<-0.2114 -5.6449 -0.827>, rot=<0 0 0 -1>}
 deform_elbow_in_r {parent=deform_elbow_in_base_r, loc=<0 0 0>, rot=<0 0 0 1>}
 deform_elbow_out_base_r {parent=lowerarm_r, loc=<-0.0001 7.5129 0.0002>, rot=<0 0 0 -1>}
 deform_elbow_out_r {parent=deform_elbow_out_base_r, loc=<0 0 0>, rot=<0 0 0 1>}
 MSTR_handle {parent=lowerarm_r, loc=<-0.737 5.808 -0.8104>, rot=<0.0265 -0.0078 -0.0511 -0.9983>}
 dyn_handle {parent=MSTR_handle, loc=<0 0 0>, rot=<0 0 0 1>}
 robot_upperarm_r {parent=upperarm_r, loc=<-3.3769 -0.2443 -0.8518>, rot=<-0.0065 -0.0392 0.0192 -0.999>}
 base_shoulderpad_r {parent=robot_upperarm_r, loc=<-1.2436 0.126 -4.9442>, rot=<0.0065 0.0392 -0.0192 -0.999>}
 shoulderpad_r {parent=base_shoulderpad_r, loc=<0 0 0>, rot=<0 0 0 1>}
 MSTR_arm_strap_01_r {parent=robot_upperarm_r, loc=<-10.7848 -0.1221 7.0349>, rot=<0.5387 -0.2998 0.6939 -0.3722>}
 dyn_arm_strap_01_r {parent=MSTR_arm_strap_01_r, loc=<0 0 0>, rot=<0 0 0 1>}
 robot_arm_transition_r {parent=robot_upperarm_r, loc=<0 0 0>, rot=<0 0 0 1>}
 attach_cape {parent=spine_05, loc=<9.017 -11.714 0>, rot=<0.0616 -0.7044 -0.0616 -0.7044>}
 cape {parent=attach_cape, type=Null, loc=<0 -12.3566 0>, rot=<0.1308 0 0 0.9914>}
 attach_backpack {parent=spine_05, loc=<-0.123 -13.473 0>, rot=<-0.0093 -0.707 0.0093 -0.707>}
 Backpack_BR {parent=attach_backpack, type=Null, loc=<0 -8.253 0>, rot=<-0.0392 0 0 0.9992>}
 MSTR_chest_tube {parent=spine_05, loc=<4.6274 -17.7672 -21.166>, rot=<-0.0253 0.7067 0.7071 -0.0008>}
 dyn_chest_tube_01 {parent=MSTR_chest_tube, loc=<0 0 0>, rot=<0 0 0 1>}
 dyn_chest_tube_02 {parent=dyn_chest_tube_01, loc=<22.0376 0 0>, rot=<-0.0815 0.1801 -0.0928 -0.9759>}
 dyn_chest_tube_03 {parent=dyn_chest_tube_02, loc=<21.0243 0 0>, rot=<-0.0066 0.4727 0.0035 -0.8812>}
 dyn_chest_tube_04 {parent=dyn_chest_tube_03, loc=<11.8473 0 0>, rot=<0.0867 0.1554 -0.035 -0.9834>}
 cardboard_back {parent=spine_05, loc=<-11.9583 -16.6264 -0.2865>, rot=<0 0 0 1>}
 cardboard_flap {parent=spine_05, loc=<14.6432 -15.0978 -0.164>, rot=<0 0 0 1>}
 deform_pec_aim_base_l {parent=spine_04, loc=<5.0227 13.6719 -2.6508>, rot=<0.0672 -0.5201 0.3228 -0.7879>}
 deform_pec_aim_l {parent=deform_pec_aim_base_l, loc=<0 0 0>, rot=<0 0 0 1>}
 deform_pec_base_l {parent=deform_pec_aim_l, loc=<16.3279 0 0>, rot=<0 0 0 1>}
 deform_pec_l {parent=deform_pec_base_l, loc=<0 0 0>, rot=<0 0 0 1>}
 deform_pec_aim_base_r {parent=spine_04, loc=<5.0225 13.6718 2.6508>, rot=<-0.5201 -0.0672 0.7879 0.3228>}
 deform_pec_aim_r {parent=deform_pec_aim_base_r, loc=<0 0 0>, rot=<0 0 0 1>}
 deform_pec_base_r {parent=deform_pec_aim_r, loc=<-16.3279 0 0>, rot=<0 0 0 -1>}
 deform_pec_r {parent=deform_pec_base_r, loc=<0 0 0.0002>, rot=<0 0 0 -1>}
 deform_lat_l_base {parent=spine_03, loc=<2.5561 -10.3716 -17.6918>, rot=<0 0 0 -1>}
 deform_lat_l {parent=deform_lat_l_base, loc=<0 0 0>, rot=<0 0 0 1>}
 deform_lat_r_base {parent=spine_03, loc=<2.5562 -10.3716 17.6918>, rot=<0 0 1 0>}
 deform_lat_r {parent=deform_lat_r_base, loc=<0 0 0>, rot=<0 0 0 -1>}
 power_strip {parent=spine_02, loc=<1.0387 19.5991 1.1743>, rot=<-0.0351 -0.0051 -0.1438 -0.989>}
 MSTR_chest_tube_reverse {parent=power_strip, loc=<-6.2282 3.4575 -6.1037>, rot=<0.1434 -0.8449 0.2076 -0.4717>}
 dyn_chest_tube_01_reverse {parent=MSTR_chest_tube_reverse, loc=<0 0 0>, rot=<0 0 0 1>}
 dyn_chest_tube_02_reverse {parent=dyn_chest_tube_01_reverse, loc=<19.7422 0 0>, rot=<-0.1627 0.0017 0.445 -0.8806>}
 dyn_chest_tube_03_reverse {parent=dyn_chest_tube_02_reverse, loc=<11.9573 0 0>, rot=<0.0026 0.0419 0.0608 -0.9973>}
 thigh_l {parent=pelvis, loc=<-1.8716 -0.0109 -11.9936>, rot=<0.0012 0.0572 0.0155 -0.9982>}
 calf_l {parent=thigh_l, loc=<-41.9178 -0.2246 0.0926>, rot=<0.0006 -0.0161 -0.0391 -0.9991>}
 foot_l {parent=calf_l, loc=<-38.7057 1.2438 0.2452>, rot=<0.0975 -0.0474 0.0317 -0.9936>}
 ball_l {parent=foot_l, loc=<-6.1614 13.1469 0.6807>, rot=<0.0001 -0.0001 -0.7063 -0.7079>}
 calf_twist_02_l {parent=calf_l, loc=<-15.9268 -0.0967 0.0397>, rot=<0.0237 0.0095 -0.0035 -0.9997>}
 MSTR_boot_tassel_l {parent=calf_twist_02_l, loc=<1.7617 -11.1864 -0.3603>, rot=<-0.006 -0.0123 0.2323 -0.9725>}
 dyn_boot_tassel_l {parent=MSTR_boot_tassel_l, loc=<0 0 0>, rot=<0 0 0 1>}
 calf_twist_01_l {parent=calf_l, loc=<-32.9335 -0.0967 0.0397>, rot=<0.0237 0.0095 -0.0035 -0.9997>}
 MSTR_boot_strap_l {parent=calf_twist_01_l, loc=<-0.7468 6.3973 5.2891>, rot=<0.459 -0.3433 0.819 -0.0261>}
 dyn_boot_strap_l {parent=MSTR_boot_strap_l, loc=<0 0 0>, rot=<0 0 0 1>}
 deform_knee_in_base_l {parent=calf_l, loc=<1.298 -8.0356 -0.0496>, rot=<0 0 0 -1>}
 deform_knee_in_l {parent=deform_knee_in_base_l, loc=<0 0 0>, rot=<0 0 0 1>}
 deform_knee_out_base_l {parent=calf_l, loc=<0 6.8739 0>, rot=<0 0 0 -1>}
 deform_knee_out_l {parent=deform_knee_out_base_l, loc=<0 0 0>, rot=<0 0 0 1>}
 base_kneepad_l {parent=calf_l, loc=<3.2402 8.1305 1.2771>, rot=<0.48 0.503 0.5311 0.4843>}
 kneepad_l {parent=base_kneepad_l, loc=<0 0 0>, rot=<0 0 0 1>}
 thigh_twist_01_l {parent=thigh_l, loc=<-21.6303 0 0>, rot=<0.0013 0 0 -1>}
 deform_groin_base_l {parent=thigh_l, loc=<-2.1908 8.5661 -1.7442>, rot=<0 0 0 -1>}
 deform_groin_l {parent=deform_groin_base_l, loc=<0 0 0>, rot=<0 0 0 1>}
 thigh_r {parent=pelvis, loc=<-1.8716 -0.0109 11.9936>, rot=<0.0572 -0.0012 0.9982 0.0155>}
 calf_r {parent=thigh_r, loc=<41.9178 0.2246 -0.0926>, rot=<0.0006 -0.0161 -0.0391 -0.9991>}
 foot_r {parent=calf_r, loc=<38.7057 -1.2438 -0.2452>, rot=<0.0975 -0.0474 0.0317 -0.9936>}
 ball_r {parent=foot_r, loc=<6.1614 -13.1469 -0.6807>, rot=<0.0001 -0.0001 -0.7063 -0.7079>}
 calf_twist_02_r {parent=calf_r, loc=<15.9268 0.0967 -0.0397>, rot=<0.0237 0.0095 -0.0035 -0.9997>}
 calf_twist_01_r {parent=calf_r, loc=<32.9335 0.0967 -0.0397>, rot=<0.0237 0.0095 -0.0035 -0.9997>}
 deform_knee_in_base_r {parent=calf_r, loc=<-1.2981 8.0356 0.0496>, rot=<0 0 0 -1>}
 deform_knee_in_r {parent=deform_knee_in_base_r, loc=<0 0 0>, rot=<0 0 0 1>}
 deform_knee_out_base_r {parent=calf_r, loc=<0 -6.8739 0>, rot=<0 0 0 -1>}
 deform_knee_out_r {parent=deform_knee_out_base_r, loc=<0 0 0>, rot=<0 0 0 1>}
 thigh_twist_01_r {parent=thigh_r, loc=<21.6303 0 0>, rot=<0.0013 0 0 -1>}
 deform_groin_base_r {parent=thigh_r, loc=<2.1908 -8.5661 1.7442>, rot=<0 0 0 -1>}
 deform_groin_r {parent=deform_groin_base_r, loc=<0 0 0>, rot=<0 0 0 -1>}
 deform_glute_base_l {parent=pelvis, loc=<-10.1877 -11.6217 -12.899>, rot=<0.0275 0.0501 -0.4472 -0.8926>}
 deform_glute_l {parent=deform_glute_base_l, loc=<0 0 0>, rot=<0 0 0 1>}
 deform_glute_base_r {parent=pelvis, loc=<-10.1877 -11.6218 12.899>, rot=<0.0501 -0.0275 0.8926 -0.4472>}
 deform_glute_r {parent=deform_glute_base_r, loc=<0 0 0>, rot=<0 0 0 1>}
 MSTR_pliers {parent=pelvis, loc=<3.6771 -14.1887 9.0672>, rot=<-0.5221 0.0313 0.0311 -0.8517>}
 dyn_pliers {parent=MSTR_pliers, loc=<0 0 0>, rot=<0 0 0 1>}
 MSTR_pouch_01_l {parent=pelvis, loc=<6.6348 9.5488 -15.8589>, rot=<-0.0287 -0.9567 -0.2889 -0.0205>}
 dyn_pouch_01_l {parent=MSTR_pouch_01_l, loc=<0 0 0>, rot=<0 0 0 1>}
 dyn_pouch_02_l {parent=dyn_pouch_01_l, loc=<12.3025 0 0>, rot=<0.0023 0.021 0.0165 -0.9996>}
 MSTR_screwdriver {parent=dyn_pouch_01_l, loc=<8.4529 0.4503 1.6611>, rot=<0.0496 -0.9941 0.0445 0.0858>}
 dyn_screwdriver {parent=MSTR_screwdriver, loc=<0 0 0>, rot=<0 0 0 1>}
 MSTR_pouch_01_r {parent=pelvis, loc=<6.6348 9.5488 15.8589>, rot=<0.9567 -0.0287 -0.0205 0.2889>}
 dyn_pouch_01_r {parent=MSTR_pouch_01_r, loc=<0 0 0>, rot=<0 0 0 1>}
 dyn_pouch_02_r {parent=dyn_pouch_01_r, loc=<-12.3025 0 0>, rot=<0.0023 0.021 0.0165 -0.9996>}
 MSTR_wrench {parent=dyn_pouch_01_r, loc=<-3.9661 -0.2511 -0.6055>, rot=<-0.0365 -0.0406 -0.2165 -0.9747>}
 dyn_wrench {parent=MSTR_wrench, loc=<0 0 0>, rot=<0 0 0 1>}
]
variables [
 FE idle {type=bool, default=False}
 melee {type=bool, default=False}
 targeting {type=bool, default=False}
]
