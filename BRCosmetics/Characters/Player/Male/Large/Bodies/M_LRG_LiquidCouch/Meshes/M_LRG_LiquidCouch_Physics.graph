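# Physics asset for the LiquidCouch large male player body cosmetic, giving capsule bodies to its dangling gear and nearby body parts.
# Capsules cover the pelvis, spine, left robot arm, shoulder pad and left thigh, plus dyn_ bones for chest tubes, pouch and screwdriver.
# Tubes 01–03 chain from spine_05 with tight swing and zero twist limits, while tube_04 hangs separately from the pelvis.
# The screwdriver hangs off the pelvis-mounted pouch, and most other joints swing and twist freely.
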PhysicsAsset /BRCosmetics/Characters/Player/Male/Large/Bodies/M_LRG_LiquidCouch/Meshes/M_LRG_LiquidCouch_Physics
bodies {
 dyn_chest_tube_01 [{
   shapeType=Capsule
   length=10
   radius=2.5
   shapeName=dyn_chest_tube_01_capsule
   center=<8.6757 -1.3267 -0.4663>
   rotation=<rot -0.5975 -92.6015 -89.0073>
  }]
 dyn_chest_tube_02 [{
   shapeType=Capsule
   length=14.9393
   radius=2.5
   shapeName=dyn_chest_tube_02_capsule
   center=<7.7891 0.3067 -1.0344>
   rotation=<rot 0.3108 -90.8602 -83.1206>
  }]
 dyn_chest_tube_03 [{
   shapeType=Capsule
   length=20.8536
   radius=2.582
   shapeName=dyn_chest_tube_03_capsule
   center=<9.2154 -1.3471 -0.433>
   rotation=<rot 8.9642 -87.7389 -96.9701>
  }]
 dyn_chest_tube_04 [{
   shapeType=Capsule
   length=18.2628
   radius=2
   shapeName=dyn_chest_tube_04_capsule
   center=<3.5526 -0.4727 0.3128>
   rotation=<rot 82.3571 -169.2027 -176.8887>
  }]
 dyn_pouch_01_l [{
   shapeType=Capsule
   length=13.792
   radius=9.6175
   shapeName=dyn_pouch_01_l_capsule_B
   center=<8.5623 -0.6571 -8.0454>
   rotation=<rot 1.2248 -92.5898 -98.5914>
  }]
 dyn_screwdriver [{
   shapeType=Capsule
   length=14.627
   radius=2.0033
   shapeName=dyn_screwdriver_capsule
   center=<6.2195 -0.0313 1.1342>
   rotation=<rot 0 89.8499 -94.2029>
  }]
 pelvis [{shapeType=Capsule, length=1.14, radius=19.0647, center=<0 1.3536 0>, rotation=<rot 90 0 0>}]
 robot_lowerarm_l [{shapeType=Capsule, length=12, radius=7, shapeName=robot_lowerarm_l_capsule, center=<6.4398 0 0>, rotation=<rot 90 0 0>}]
 robot_upperarm_l [{shapeType=Capsule, length=21.2071, radius=7.4778, shapeName=robot_upperarm_l_capsule, center=<9.763 0 0>, rotation=<rot 90 0 0>}]
 shoulderpad_l [{shapeType=Capsule, length=12, radius=10, shapeName=shoulderpad_l_capsule, center=<0.7647 1.5672 -1.93>, rotation=<rot 0 1.105 -89.5757>}]
 spine_01 [{shapeType=Capsule, length=4.4038, radius=15.511, shapeName=spine_01_capsule, center=<-0.0914 -2.8842 -2.2151>, rotation=<rot 0 -90 0>}]
 spine_02 [
  {
   shapeType=Capsule
   length=18.7239
   radius=11.3121
   shapeName=spine_02_capsule
   center=<1.7009 -6.1358 -0.7751>
   rotation=<rot -6.1406 -90 0.4753>
  }
  {
   shapeType=Capsule
   length=12.7566
   radius=11.3121
   shapeName=spine_02_capsule
   center=<2.1947 7.3468 0.7547>
   rotation=<rot -6.1406 -90 0.4753>
  }
 ]
 spine_03 [{shapeType=Capsule, length=22.6607, radius=11.3005, shapeName=spine_03_capsule, center=<4.6451 -6.3116 0>, rotation=<rot 0 -90 0>}]
 spine_05 [{
   shapeType=Capsule
   length=27.3658
   radius=11.8575
   shapeName=spine_05_capsule
   center=<-1.8653 -5.9462 1.0411>
   rotation=<rot -0.1972 -89.8567 -1.7935>
  }]
 thigh_l [{
   shapeType=Capsule
   length=28.7553
   radius=11.2346
   shapeName=thigh_l_capsule
   center=<-10.4893 -0.9242 3.9875>
   rotation=<rot 87.681 179.9992 179.9994>
  }]
}
constraints [
 {bone1Name=spine_02, bone2Name=spine_01, swing1LimitDegrees=45, swing2LimitDegrees=45, twistLimitDegrees=45}
 {bone1Name=dyn_pouch_01_l, bone2Name=pelvis, swing1LimitDegrees=45, swing2LimitDegrees=45, twistLimitDegrees=45}
 {
  bone1Name=spine_03
  bone2Name=spine_02
  swing1LimitDegrees=45
  swing1Motion=Free
  swing2LimitDegrees=45
  swing2Motion=Free
  twistLimitDegrees=45
  twistMotion=Free
 }
 {
  bone1Name=spine_05
  bone2Name=spine_03
  swing1LimitDegrees=45
  swing1Motion=Free
  swing2LimitDegrees=45
  swing2Motion=Free
  twistLimitDegrees=45
  twistMotion=Free
 }
 {
  bone1Name=spine_01
  bone2Name=pelvis
  swing1LimitDegrees=45
  swing1Motion=Free
  swing2LimitDegrees=45
  swing2Motion=Free
  twistLimitDegrees=45
  twistMotion=Free
 }
 {bone1Name=dyn_chest_tube_01, bone2Name=spine_05, swing1LimitDegrees=10, swing2LimitDegrees=10, twistLimitDegrees=0}
 {bone1Name=dyn_chest_tube_02, bone2Name=dyn_chest_tube_01, swing1LimitDegrees=20, swing2LimitDegrees=20, twistLimitDegrees=0}
 {bone1Name=dyn_chest_tube_03, bone2Name=dyn_chest_tube_02, swing1LimitDegrees=20, swing2LimitDegrees=20, twistLimitDegrees=0}
 {
  bone1Name=robot_upperarm_l
  bone2Name=spine_05
  swing1LimitDegrees=45
  swing1Motion=Free
  swing2LimitDegrees=45
  swing2Motion=Free
  twistLimitDegrees=45
  twistMotion=Free
 }
 {
  bone1Name=robot_lowerarm_l
  bone2Name=robot_upperarm_l
  swing1LimitDegrees=45
  swing1Motion=Free
  swing2LimitDegrees=45
  swing2Motion=Free
  twistLimitDegrees=45
  twistMotion=Free
 }
 {
  bone1Name=shoulderpad_l
  bone2Name=robot_upperarm_l
  swing1LimitDegrees=45
  swing1Motion=Free
  swing2LimitDegrees=45
  swing2Motion=Free
  twistLimitDegrees=45
  twistMotion=Free
 }
 {
  bone1Name=dyn_screwdriver
  bone2Name=dyn_pouch_01_l
  swing1LimitDegrees=45
  swing1Motion=Free
  swing2LimitDegrees=45
  swing2Motion=Free
  twistLimitDegrees=45
  twistMotion=Free
 }
 {
  bone1Name=thigh_l
  bone2Name=pelvis
  swing1LimitDegrees=45
  swing1Motion=Free
  swing2LimitDegrees=45
  swing2Motion=Free
  twistLimitDegrees=45
  twistMotion=Free
 }
 {
  bone1Name=dyn_chest_tube_04
  bone2Name=pelvis
  swing1LimitDegrees=45
  swing1Motion=Free
  swing2LimitDegrees=45
  swing2Motion=Free
  twistLimitDegrees=45
  twistMotion=Free
 }
]
