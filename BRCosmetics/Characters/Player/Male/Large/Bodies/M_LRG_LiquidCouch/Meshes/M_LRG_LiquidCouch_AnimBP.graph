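# Post-process style AnimBP for the LiquidCouch large male body, layering secondary motion onto the pose copied from its parent mesh.
# Control Rig feeds AnimDynamics for straps, tassel and tools, a backbling-driven cardboard flap offset, then RigidBody physics.
# Outside front-end emotes, the frontend_idle_bypass curve weights a DefaultSlot blend on arms and shoulder pads.
# The event graph updates these states and computes gravity-override parameters.
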AnimBlueprint /BRCosmetics/Characters/Player/Male/Large/Bodies/M_LRG_LiquidCouch/Meshes/M_LRG_LiquidCouch_AnimBP
target_skeleton=&/M_LRG_LiquidCouch_Skeleton
graphs [
 AnimGraph {nodes [
   n1 = Layered blend per bone(
    BasePose=Use cached pose 'RBAN_Cache'()
    BlendPoses_0=Slot 'DefaultSlot'()
    BlendWeights_0=Get Custom Idle is Playing Float() -> (Custom Idle is Playing Float:double)
   ) {layerSetup [{branchFilters [{boneName=base_shoulderpad_l}, {boneName=base_shoulderpad_r}, {boneName=arm_01_l}, {boneName=arm_01_r}]}]}
   Output Pose(Result=Blend Poses by bool(BlendPose_0=Use cached pose 'RBAN_Cache'(), BlendPose_1=@n1, bActiveValue=Get isFE_EmotePlay() -> (isFE_EmotePlay:bool))) {layerGroup=None}
   n2 = Control Rig!(
    Source=Copy Pose From Mesh() {bUseAttachedParent=true}
    Alpha
    melee:bool=Get bIsPlayingMeleeAnim()
    targeting:bool=Get bIsTargeting()
    FE idle:bool=Get Custom Idle Is Playing() -> (Custom Idle Is Playing:bool)
   ) -> (Pose) {controlRigAssetReference {blueprintRigClass=&/M_LRG_LiquidCouch_CtrlRig.M_LRG_LiquidCouch_CtrlRig_C}}
   n3 = Anim Dynamics - Bone: dyn_arm_strap_01_r!(ComponentPose=Local To Component(LocalPose=@n2), Alpha=0.5, GravityOverride:Vector="0,0,-980", ExternalForce="0,0,0") -> (Pose) {
    angularDampingOverride=0.9999
    angularSpringConstant=8000
    bAngularSpring=true
    bOverrideAngularDamping=true
    bUseGravityOverride=true
    bUsePlanarLimit=false
    gravityScale=0.35
    simulationSpace=RootRelative
    alphaBoolBlend {blendOption=Linear}
    boundBone {boneName=dyn_arm_strap_01_r}
    componentAppliedLinearAccClamp=<1400 1400 1400>
    componentLinearAccScale=<1 1 1>
    componentLinearVelScale=<1 1 1>
    gravityOverride {z=-980}
    physicsBodyDefinitions [{
      boundBone {boneName=dyn_arm_strap_01_r}
      boxExtents=<20 15 10>
      constraintSetup {angularLimitsMax=<0 45 45>, angularLimitsMin=<0 -45 -45>, angularTarget=<1 0 0>}
      localJointOffset=<4 0 1>
     }]
   }
   n4 = Anim Dynamics - Bone: dyn_arm_strap_02_r!(ComponentPose=@n3, Alpha=0.5, GravityOverride:Vector="0,0,-980", ExternalForce="0,0,0") -> (Pose) {
    angularDampingOverride=0.9999
    angularSpringConstant=8000
    bAngularSpring=true
    bOverrideAngularDamping=true
    bUseGravityOverride=true
    bUsePlanarLimit=false
    gravityScale=0.35
    simulationSpace=RootRelative
    alphaBoolBlend {blendOption=Linear}
    boundBone {boneName=dyn_arm_strap_02_r}
    componentAppliedLinearAccClamp=<1400 1400 1400>
    componentLinearAccScale=<1 1 1>
    componentLinearVelScale=<1 1 1>
    gravityOverride {z=-980}
    physicsBodyDefinitions [{
      boundBone {boneName=dyn_arm_strap_02_r}
      boxExtents=<20 15 10>
      constraintSetup {angularLimitsMax=<0 45 45>, angularLimitsMin=<0 -45 -45>, angularTarget=<1 0 0>}
      localJointOffset=<4 0 1>
     }]
   }
   n5 = Anim Dynamics - Bone: dyn_boot_strap_l!(ComponentPose=@n4, Alpha=0.25, GravityOverride:Vector="0,0,-980", ExternalForce="0,0,0") -> (Pose) {
    angularDampingOverride=0.9999
    angularSpringConstant=8000
    bAngularSpring=true
    bOverrideAngularDamping=true
    bUseGravityOverride=true
    bUsePlanarLimit=false
    gravityScale=0.35
    simulationSpace=RootRelative
    alphaBoolBlend {blendOption=Linear}
    boundBone {boneName=dyn_boot_strap_l}
    componentAppliedLinearAccClamp=<1400 1400 1400>
    componentLinearAccScale=<1 1 1>
    componentLinearVelScale=<1 1 1>
    gravityOverride {z=-980}
    physicsBodyDefinitions [{
      boundBone {boneName=dyn_boot_strap_l}
      boxExtents=<20 15 10>
      constraintSetup {angularLimitsMax=<0 45 45>, angularLimitsMin=<0 -45 -45>, angularTarget=<1 0 0>}
      localJointOffset=<4 0 1>
     }]
   }
   n6 = Anim Dynamics - Bone: dyn_boot_tassel_l!(ComponentPose=@n5, Alpha=0.5, GravityOverride:Vector="0,0,-980", ExternalForce="0,0,0") -> (Pose) {
    angularDampingOverride=0.9999
    angularSpringConstant=8000
    bAngularSpring=true
    bOverrideAngularDamping=true
    bUseGravityOverride=true
    bUsePlanarLimit=false
    gravityScale=0.35
    simulationSpace=RootRelative
    alphaBoolBlend {blendOption=Linear}
    boundBone {boneName=dyn_boot_tassel_l}
    componentAppliedLinearAccClamp=<1400 1400 1400>
    componentLinearAccScale=<1 1 1>
    componentLinearVelScale=<1 1 1>
    gravityOverride {z=-980}
    physicsBodyDefinitions [{
      boundBone {boneName=dyn_boot_tassel_l}
      boxExtents=<20 15 10>
      constraintSetup {angularLimitsMax=<0 45 45>, angularLimitsMin=<0 -45 -45>, angularTarget=<1 0 0>}
      localJointOffset=<4 0 1>
     }]
   }
   n7 = Anim Dynamics - Bone: dyn_wrench!(ComponentPose=@n6, Alpha=0.5, GravityOverride:Vector="0,0,-980", ExternalForce="0,0,0") -> (Pose) {
    angularDampingOverride=1
    angularSpringConstant=8000
    bAngularSpring=true
    bOverrideAngularDamping=true
    bUseGravityOverride=true
    bUsePlanarLimit=false
    gravityScale=0.35
    simulationSpace=RootRelative
    alphaBoolBlend {blendOption=Linear}
    boundBone {boneName=dyn_wrench}
    componentAppliedLinearAccClamp=<1400 1400 1400>
    componentLinearAccScale=<1 1 1>
    componentLinearVelScale=<1 1 1>
    gravityOverride {z=-980}
    physicsBodyDefinitions [{
      boundBone {boneName=dyn_wrench}
      boxExtents=<20 15 10>
      constraintSetup {angularLimitsMax=<0 10 10>, angularLimitsMin=<0 -10 -10>, angularTarget=<1 0 0>}
      localJointOffset=<4 0 1>
     }]
   }
   n8 = Anim Dynamics - Bone: dyn_screwdriver!(ComponentPose=@n7, Alpha=0.5, GravityOverride:Vector="0,0,-980", ExternalForce="0,0,0") -> (Pose) {
    angularDampingOverride=1
    angularSpringConstant=8000
    bAngularSpring=true
    bOverrideAngularDamping=true
    bUseGravityOverride=true
    bUsePlanarLimit=false
    gravityScale=0.35
    simulationSpace=RootRelative
    alphaBoolBlend {blendOption=Linear}
    boundBone {boneName=dyn_screwdriver}
    componentAppliedLinearAccClamp=<1400 1400 1400>
    componentLinearAccScale=<1 1 1>
    componentLinearVelScale=<1 1 1>
    gravityOverride {z=-980}
    physicsBodyDefinitions [{
      boundBone {boneName=dyn_screwdriver}
      boxExtents=<20 15 10>
      constraintSetup {angularLimitsMax=<0 10 10>, angularLimitsMin=<0 -10 -10>, angularTarget=<1 0 0>}
      localJointOffset=<4 0 1>
     }]
   }
   n9 = Anim Dynamics - Bone: dyn_pliers!(ComponentPose=@n8, Alpha=0.25, GravityOverride:Vector="0,0,-980", ExternalForce="0,0,0") -> (Pose) {
    angularDampingOverride=1
    angularSpringConstant=8000
    bAngularSpring=true
    bOverrideAngularDamping=true
    bUseGravityOverride=true
    bUsePlanarLimit=false
    gravityScale=0.35
    simulationSpace=RootRelative
    alphaBoolBlend {blendOption=Linear}
    boundBone {boneName=dyn_pliers}
    componentAppliedLinearAccClamp=<1400 1400 1400>
    componentLinearAccScale=<1 1 1>
    componentLinearVelScale=<1 1 1>
    gravityOverride {z=-980}
    physicsBodyDefinitions [{
      boundBone {boneName=dyn_pliers}
      boxExtents=<20 15 10>
      constraintSetup {angularLimitsMax=<0 0 10>, angularLimitsMin=<0 0 -10>, angularTarget=<1 0 0>}
      localJointOffset=<4 0 1>
     }]
   }
   n10 = RigidBody!(
    ComponentPose="Transform (Modify) Bone - Bone: cardboard_flap"(ComponentPose=@n9, Alpha=Get backbling() -> (backbling:double), Translation="0,0,0", Rotation="0,-25,0", Scale="1,1,1") {rotationMode=BMM_Additive, rotationSpace=BCS_BoneSpace, boneToModify {boneName=cardboard_flap}, rotation {yaw=-25}}
    Alpha
    OverrideWorldGravity:Vector="0,0,0"
    ExternalForce
    SimSpaceSettings:Sim Space Settings Structure=(WorldAlpha=1,VelocityScaleZ=0.4,DampingAlpha=1,MaxLinearVelocity=300,MaxAngularVelocity=10,MaxLinearAcceleration=1000,MaxAngularAcceleration=40,ExternalLinearDragV=(X=0.3,Y=0.3,Z=0.3),ExternalLinearVelocity=(X=0,Y=0,Z=0),ExternalAngularVelocity=(X=0,Y=0,Z=0))
   ) -> (Pose) {
    bClampLinearTranslationLimitToRefPose=true
    bOverrideWorldGravity=true
    cachedBoundsScale=1
    overridePhysicsAsset=&/M_LRG_LiquidCouch_Physics
    simulationSpace=BaseBoneSpace
    alphaBoolBlend {blendOption=Linear}
    baseBoneRef {boneName=spine_05}
    componentAppliedLinearAccClamp=<0 0 0>
    overrideWorldGravity {z=-980}
    simSpaceSettings {
     maxAngularAcceleration=40
     maxAngularVelocity=10
     maxLinearAcceleration=1000
     maxLinearVelocity=300
     velocityScaleZ=0.4
     worldAlpha=1
     externalLinearDragV=<0.3 0.3 0.3>
    }
   }
   Save cached pose 'RBAN_Cache'(Pose=Component To Local(ComponentPose=@n10)) {cacheName=RBAN_Cache}
  ]}
 EventGraph {nodes [
   Event BlueprintUpdateAnimation()
   > n1 = Sequence()
   > n2 = "Gravity Override State Function (Parameter Asset) (Nativized)"(
    CharacterPartAnimInstance=Self-Reference()
    GameTimeInSeconds=GetGameTimeInSeconds()
    DebugJointName=spine_05
    Parameters=&/M_LRG_LiquidCouch_Pack_Parameters
   )
   > Set Gravity Override(Gravity Override=@n2)
   > Set RBAN Settings Out(RBAN Settings Out=@n2.OutRigidBodyAnimNodeParameters)
   > Set backbling(backbling:double="To Float (Boolean)"(InBool=CheckForBackbling())) -> (Output_Get:double)
   > Set isFE_EmotePlay(isFE_EmotePlay:bool=AND Boolean(A=Get bIsInFrontEnd(), B=Get bIsPlayingEmote())) -> (Output_Get:bool)
   Branch(execute=@n1.then_1, Condition=Get bIsInFrontEnd())
   > Set Custom Idle Is Playing Float(Custom Idle Is Playing Float:double=GetCurveValue(self=GetOwnerFrontEndAnimInstance(), CurveName=frontend_idle_bypass)) -> (Output_Get:double)
   > n3 = Branch(Condition="float >= float"(A=Get Custom Idle Is Playing Float() -> (Custom Idle Is Playing Float:double), B=1))
   > Set Custom Idle Is Playing(Custom Idle Is Playing:bool=true) -> (Output_Get:bool)
   Set Custom Idle Is Playing(execute=@n3.else, Custom Idle Is Playing=false)
   AND Boolean(A=Get bIsInFrontEnd(), B=NOT Boolean(A=Get bIsPlayingEmote()))
  ]}
]
variables [
 backbling {type=number}
 Custom Idle Is Playing {type=boolean}
 Custom Idle Is Playing Float {type=number}
 Gravity Override {type=Vector}
 isFE_EmotePlay {type=boolean}
 RBAN Settings Out {
  type=FortRigidBodyAnimNodeParameters
  default {simSpaceSettings {dampingAlpha=0, maxAngularAcceleration=16, maxAngularVelocity=4, maxLinearAcceleration=800, maxLinearVelocity=200, velocityScaleZ=0.5}}
 }
]
