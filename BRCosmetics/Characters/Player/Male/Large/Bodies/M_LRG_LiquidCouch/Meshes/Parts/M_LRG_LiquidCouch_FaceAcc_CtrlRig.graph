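# Control Rig for the LiquidCouch large male body's face accessory, animating small arms, head tube, seat, goggles and seatbelts.
# Smoothed weights from upper-arm deform attributes, movement, jumping, targeting and melee variables blend arm and finger poses.
# Construction adds reference nulls; constrained nulls feed a Hermite spline that the head tube stretches along.
# A swing/twist helper remaps head and neck rotation into weights for corrective seat, goggle, seatbelt and tube offsets.
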ControlRigBlueprint /BRCosmetics/Characters/Player/Male/Large/Bodies/M_LRG_LiquidCouch/Meshes/Parts/M_LRG_LiquidCouch_FaceAcc_CtrlRig
graphs [
 New Function_ContainedGraph {nodes [
   RigVMFunctionEntryNode()
   n1 = &RigVMFunction_MathBoolToFloat(Value=RigVMVariableNode() {variable=jumping}.Value)
   n2 = RigVMDispatchNode(BoneName=root, Default=0, Name=deformRig_l_upperarm_up_alpha) {template="DISPATCH_RigDispatch_GetAnimAttribute(in Name,in BoneName,in Default,out Value,out Found)"}
   n3 = RigVMVariableNode() {variable=Character Forward}
   Maximum = &RigVMFunction_MathDoubleMax(
    A=RigVMDispatchNode(BoneName=root, Default=0, Name=deformRig_l_upperarm_fr_alpha) {template="DISPATCH_RigDispatch_GetAnimAttribute(in Name,in BoneName,in Default,out Value,out Found)"}.Value
    B=@n3.Value
   )
   n4 = RigVMVariableNode() {variable=Character Move Left}
   n5 = &RigVMFunction_AlphaInterp(
    InterpSpeedDecreasing=10
    InterpSpeedIncreasing=20
    Scale=1
    bClampResult=False
    Value=&RigVMFunction_MathDoubleMax(
     A=&RigVMFunction_MathDoubleMax(B=@n1.Result, A=@n2.Value).Result
     B=&RigVMFunction_MathDoubleMax(B=@Maximum.Result, A=@n4.Value).Result
    ).Result
   )
   > &RigUnit_ModifyTransforms(
    ItemToModify="((Item=(Type=Bone,Name=\"arm_01_l\"),Transform=(Rotation=(X=-0.1039,Y=-0.0899,Z=0.3738,W=0.9173),Translation=(X=1,Y=-1,Z=1),Scale3D=(X=1,Y=1,Z=1))),(Item=(Type=Bone,Name=\"arm_02_l\"),Transform=(Rotation=(X=0.2122,Y=-0.2442,Z=0.0547,W=0.9446),Translation=(X=0, ...(truncated)"
    WeightMinimum=0
    Weight=@n5.Result
   )
   n6 = &RigVMFunction_MathBoolToFloat(Value=RigVMVariableNode() {variable=jumping}.Value)
   n7 = RigVMDispatchNode(BoneName=root, Default=0, Name=deformRig_r_upperarm_up_alpha) {template="DISPATCH_RigDispatch_GetAnimAttribute(in Name,in BoneName,in Default,out Value,out Found)"}
   n8 = &RigVMFunction_MathDoubleMax(
    B=RigVMDispatchNode(BoneName=root, Default=0, Name=deformRig_r_upperarm_fr_alpha) {template="DISPATCH_RigDispatch_GetAnimAttribute(in Name,in BoneName,in Default,out Value,out Found)"}.Value
    A=@n3.Value
   )
   n9 = RigVMVariableNode() {variable=Character Move Right}
   n10 = &RigVMFunction_AlphaInterp(
    InterpSpeedDecreasing=10
    InterpSpeedIncreasing=20
    Scale=1
    bClampResult=False
    Value=&RigVMFunction_MathDoubleMax(
     B=&RigVMFunction_MathDoubleMax(B=@n6.Result, A=@n7.Value).Result
     A=&RigVMFunction_MathDoubleMax(A=@n8.Result, B=@n9.Value).Result
    ).Result
   )
   > &RigUnit_ModifyTransforms(
    ItemToModify="((Item=(Type=Bone,Name=\"arm_01_r\"),Transform=(Rotation=(X=-0.1039,Y=-0.0899,Z=0.3738,W=0.9173),Translation=(X=-1,Y=-1,Z=-1),Scale3D=(X=1,Y=1,Z=1))),(Item=(Type=Bone,Name=\"arm_02_r\"),Transform=(Rotation=(X=0.2629,Y=-0.0738,Z=0.1498,W=0.9502),Translation=(X= ...(truncated)"
    WeightMinimum=0
    Weight=@n10.Result
   )
   n11 = RigVMDispatchNode(BoneName=root, Default=0, Name=deformRig_l_upperarm_bk_alpha) {template="DISPATCH_RigDispatch_GetAnimAttribute(in Name,in BoneName,in Default,out Value,out Found)"}
   n12 = RigVMVariableNode() {variable=Character Backward}
   n13 = &RigVMFunction_AlphaInterp(
    InterpSpeedDecreasing=10
    InterpSpeedIncreasing=20
    Scale=1
    bClampResult=False
    Value=&RigVMFunction_MathDoubleMax(B=&RigVMFunction_MathDoubleMax(A=@n11.Value, B=@n12.Value).Result, A=RigVMVariableNode() {variable=Character Move Right}.Value).Result
   )
   > &RigUnit_ModifyTransforms(
    ItemToModify="((Item=(Type=Bone,Name=\"arm_01_l\"),Transform=(Rotation=(X=0.0114,Y=0.1828,Z=0.3044,W=0.9348),Translation=(X=0.2921,Y=-0.421,Z=-0.0686),Scale3D=(X=1,Y=1,Z=1))),(Item=(Type=Bone,Name=\"arm_02_l\"),Transform=(Rotation=(X=-0.0068,Y=0.0782,Z=0.0869,W=0.9931),Tran ...(truncated)"
    WeightMinimum=0
    Weight=@n13.Result
   )
   RigVMFunction_MathBoolNot = &RigVMFunction_MathBoolNot(Value=RigVMVariableNode() {variable=targeting}.Value)
   n14 = RigVMDispatchNode(BoneName=root, Default=0, Name=deformRig_r_upperarm_bk_alpha) {template="DISPATCH_RigDispatch_GetAnimAttribute(in Name,in BoneName,in Default,out Value,out Found)"}
   Multiply = &RigVMFunction_MathDoubleMul(
    B=&RigVMFunction_MathBoolToFloat(Value=@RigVMFunction_MathBoolNot.Result).Result
    A=&RigVMFunction_MathDoubleMax(A=&RigVMFunction_MathDoubleMax(B=@n14.Value, A=@n12.Value).Result, B=RigVMVariableNode() {variable=Character Move Left}.Value).Result
   )
   n15 = &RigVMFunction_AlphaInterp(InterpSpeedDecreasing=10, InterpSpeedIncreasing=20, Scale=1, bClampResult=False, Value=@Multiply.Result)
   > &RigUnit_ModifyTransforms(
    ItemToModify="((Item=(Type=Bone,Name=\"arm_01_r\"),Transform=(Rotation=(X=-0.0763,Y=0.1708,Z=0.4319,W=0.8823),Translation=(X=-0.6697,Y=0.5619,Z=0),Scale3D=(X=1,Y=1,Z=1))),(Item=(Type=Bone,Name=\"arm_02_r\"),Transform=(Rotation=(X=-0.0068,Y=0.0782,Z=0.0869,W=0.9931),Translat ...(truncated)"
    WeightMinimum=0
    Weight=@n15.Result
   )
   n16 = &RigVMFunction_AlphaInterp(
    InterpSpeedDecreasing=10
    InterpSpeedIncreasing=10
    Scale=1
    bClampResult=False
    Value=RigVMVariableNode() {variable=Character Move Right}.Value
   )
   > &RigUnit_ModifyTransforms(
    ItemToModify="((Item=(Type=Bone,Name=\"arm_01_r\"),Transform=(Rotation=(X=-0.1103,Y=-0.4234,Z=-0.3659,W=0.8214),Translation=(X=1.4597,Y=0.6545,Z=-0.2652),Scale3D=(X=1,Y=1,Z=1))),(Item=(Type=Bone,Name=\"arm_02_r\"),Transform=(Rotation=(X=0.1133,Y=-0.2422,Z=0.0727,W=0.9608),T ...(truncated)"
    WeightMinimum=0
    Weight=@n16.Result
   )
   n17 = &RigVMFunction_AlphaInterp(
    InterpSpeedDecreasing=10
    InterpSpeedIncreasing=10
    Scale=1
    bClampResult=False
    Value=RigVMVariableNode() {variable=Character Move Left}.Value
   )
   > &RigUnit_ModifyTransforms(
    ItemToModify="((Item=(Type=Bone,Name=\"arm_01_l\"),Transform=(Rotation=(X=0.1178,Y=-0.4153,Z=-0.4043,W=0.8064),Translation=(X=0,Y=0,Z=0),Scale3D=(X=1,Y=1,Z=1))),(Item=(Type=Bone,Name=\"arm_02_l\"),Transform=(Rotation=(X=0.0568,Y=-0.0026,Z=-0.0597,W=0.9966),Translation=(X=-0 ...(truncated)"
    WeightMinimum=0
    Weight=@n17.Result
   )
   Cast = &RigVMFunction_MathBoolToFloat(Value=RigVMVariableNode() {variable=jumping}.Value)
   n18 = &RigVMFunction_AlphaInterp(
    InterpSpeedDecreasing=10
    InterpSpeedIncreasing=10
    Scale=1
    bClampResult=False
    Value=&RigVMFunction_MathDoubleMax(
     B=@Cast.Result
     A=RigVMDispatchNode(BoneName=root, Default=0, Name=deformRig_l_upperarm_up_alpha) {template="DISPATCH_RigDispatch_GetAnimAttribute(in Name,in BoneName,in Default,out Value,out Found)"}.Value
    ).Result
   )
   > &RigUnit_ModifyTransforms(
    ItemToModify="((Item=(Type=Bone,Name=\"arm_01_l\"),Transform=(Rotation=(X=-0.2863,Y=-0.5336,Z=-0.4595,W=0.6498),Translation=(X=0.2387,Y=-0.4782,Z=-0.1479),Scale3D=(X=1,Y=1,Z=1))),(Item=(Type=Bone,Name=\"arm_02_l\"),Transform=(Rotation=(X=-0.3153,Y=-0.2749,Z=-0.0563,W=0.9066 ...(truncated)"
    WeightMinimum=0
    Weight=@n18.Result
   )
   n19 = &RigVMFunction_AlphaInterp(
    InterpSpeedDecreasing=10
    InterpSpeedIncreasing=10
    Scale=1
    bClampResult=False
    Value=&RigVMFunction_MathDoubleMax(
     B=@Cast.Result
     A=RigVMDispatchNode(BoneName=root, Default=0, Name=deformRig_r_upperarm_up_alpha) {template="DISPATCH_RigDispatch_GetAnimAttribute(in Name,in BoneName,in Default,out Value,out Found)"}.Value
    ).Result
   )
   > &RigUnit_ModifyTransforms(
    ItemToModify="((Item=(Type=Bone,Name=\"arm_01_r\"),Transform=(Rotation=(X=-0.2561,Y=-0.564,Z=-0.4275,W=0.6584),Translation=(X=-0.0769,Y=0.1494,Z=0.6717),Scale3D=(X=1,Y=1,Z=1))),(Item=(Type=Bone,Name=\"arm_02_r\"),Transform=(Rotation=(X=-0.2329,Y=-0.4169,Z=0.1749,W=0.861),Tr ...(truncated)"
    WeightMinimum=0
    Weight=@n19.Result
   )
   RigVMFunction_MathBoolOr = &RigVMFunction_MathBoolOr(A=RigVMVariableNode() {variable=targeting}.Value, B=RigVMVariableNode() {variable=meleeanim}.Value)
   RigVMFunction_MathBoolToFloat = &RigVMFunction_MathBoolToFloat(Value=@RigVMFunction_MathBoolOr.Result)
   n20 = &RigVMFunction_AlphaInterp(InterpSpeedDecreasing=10, InterpSpeedIncreasing=10, Scale=1, bClampResult=False, Value=@RigVMFunction_MathBoolToFloat.Result)
   > &RigUnit_ModifyTransforms(
    ItemToModify="((Item=(Type=Bone,Name=\"arm_01_r\"),Transform=(Rotation=(X=0.0858,Y=-0.173,Z=0.0151,W=0.9811),Translation=(X=0,Y=0,Z=0),Scale3D=(X=1,Y=1,Z=1))),(Item=(Type=Bone,Name=\"arm_01_l\"),Transform=(Rotation=(X=0.0858,Y=-0.173,Z=0.0151,W=0.9811),Translation=(X=0,Y=0, ...(truncated)"
    WeightMinimum=0
    Weight=@n20.Result
   )
   > &RigUnit_ModifyTransforms(
    ItemToModify="((Item=(Type=Bone,Name=\"lil_index_01_r\"),Transform=(Rotation=(X=0.0181,Y=-0.6459,Z=-0.0207,W=0.7629),Translation=(X=0.1453,Y=0.2326,Z=-0.05),Scale3D=(X=1,Y=1,Z=1))),(Item=(Type=Bone,Name=\"lil_index_02_r\"),Transform=(Rotation=(X=0,Y=-0.3827,Z=0,W=0.9239),Tr ...(truncated)"
    WeightMinimum=0
    Weight=@n20.Result
   )
   > &RigUnit_ModifyTransforms(
    ItemToModify="((Item=(Type=Bone,Name=\"lil_index_01_l\"),Transform=(Rotation=(X=0.0232,Y=0.6361,Z=-0.0189,W=-0.7711),Translation=(X=-0.2922,Y=-0.168,Z=0.0283),Scale3D=(X=1,Y=1,Z=1))),(Item=(Type=Bone,Name=\"lil_index_02_l\"),Transform=(Rotation=(X=0,Y=-0.4834,Z=0,W=0.8754), ...(truncated)"
    WeightMinimum=0
    Weight=@n20.Result
   )
   > RigVMFunctionReturnNode()
  ]}
 New Function_ContainedGraph {nodes [
   RigVMFunctionEntryNode()
   > &RigUnit_ParentConstraint(
    AdvancedSettings=(InterpolationType=Average,RotationOrderForFilter=XZY)
    Child=(Type=Null,Name="headtube_null_04")
    Parents=((Item=(Type=Bone,Name="neck_01"),Weight=0.2),(Item=(Type=Bone,Name="head"),Weight=0.8))
    Weight=1
   )
   > &RigUnit_ParentConstraint(
    AdvancedSettings=(InterpolationType=Average,RotationOrderForFilter=XZY)
    Child=(Type=Null,Name="headtube_null_03")
    Parents=((Item=(Type=Bone,Name="neck_01"),Weight=0.5),(Item=(Type=Bone,Name="head"),Weight=0.5))
    Weight=1
   )
   > &RigUnit_ParentConstraint(
    AdvancedSettings=(InterpolationType=Average,RotationOrderForFilter=XZY)
    Child=(Type=Null,Name="headtube_null_02")
    Parents=((Item=(Type=Bone,Name="spine_05"),Weight=0.6),(Item=(Type=Bone,Name="head"),Weight=0.4))
    Weight=1
   )
   > &RigUnit_ParentConstraint(
    AdvancedSettings=(InterpolationType=Average,RotationOrderForFilter=XZY)
    Child=(Type=Null,Name="tube_pv_null")
    Parents=((Item=(Type=Bone,Name="spine_05"),Weight=0.4),(Item=(Type=Bone,Name="head"),Weight=0.6))
    Weight=1
   )
   > DeltaRotation v4_4 = RigVMFunctionReferenceNode(
    Rotation Order=ZYX
    child=(Type=Bone,Name="head")
    onlyTwist=true
    outputMinMax=(X=0,Y=1)
    parent=(Type=Null,Name="head_null")
    premult=1
    rotationMinMax=(X=-30,Y=-70)
    triggerAxis=X
    twistAxis=(X=1,Y=0,Z=0)
   )
   > &RigUnit_ModifyTransforms(
    ItemToModify="((Item=(Type=Null,Name=\"headtube_null_03\"),Transform=(Rotation=(X=0,Y=0,Z=0,W=1),Translation=(X=-3.745,Y=-8.4117,Z=1.941),Scale3D=(X=1,Y=1,Z=1))),(Item=(Type=Null,Name=\"headtube_null_02\"),Transform=(Rotation=(X=0,Y=0,Z=0,W=-1),Translation=(X=0,Y=-4.8476,Z= ...(truncated)"
    WeightMinimum=0
    Weight=@"DeltaRotation v4_4".outputWeight
   )
   > SplineFromItems_1 = RigVMFunctionReferenceNode(
    Items=((Type=Null,Name="headtube_null_01"),(Type=Null,Name="headtube_null_02"),(Type=Null,Name="headtube_null_03"),(Type=Null,Name="headtube_null_04"))
    Samples Per Segment=25
    Spline Mode=Hermite
   )
   Make_Array_2 = RigVMDispatchNode(Values=((Type=Bone,Name="headtube_01"),(Type=Bone,Name="headtube_02"),((),Type=Bone,Name="headtube_03"),((),Type=Bone,Name="headtube_04"))) {template="DISPATCH_RigVMDispatch_ArrayMake(in Values,out Array)"}
   > /Script/ControlRigSpline.RigUnit_FitChainToSplineCurveItemArray(
    Alignment=Stretched
    DebugSettings=(bEnabled=False,Scale=1,CurveColor=(R=1,G=1,B=0,A=1),SegmentsColor=(R=1,G=0,B=0,A=1),WorldOffset=(Rotation=(X=0,Y=0,Z=0,W=1),Translation=(X=0,Y=0,Z=0),Scale3D=(X=1,Y=1,Z=1)))
    Maximum=10
    Minimum=0
    PrimaryAxis=(X=1,Y=0,Z=0)
    RotationEaseType=QuinticEaseInOut
    Rotations=()
    SamplingPrecision=25
    SecondaryAxis=(X=0,Y=0,Z=1)
    Weight=1
    bPropagateToChildren=true
    PoleVectorPosition=&RigUnit_GetTransform(Item=(Type=Null,Name="tube_pv_null"), bInitial=False).Transform.Translation
    Items=@Make_Array_2.Array
    Spline=@SplineFromItems_1.Spline
   ) {is=struct RigVMUnitNode}
   > RigVMFunctionReturnNode()
  ]}
 New Function_ContainedGraph {nodes [
   n1 = RigVMFunctionEntryNode()
   RigVMFunction_MathBoolToInteger = &RigVMFunction_MathBoolToInteger(Value=@n1.onlyTwist)
   RigVMFunction_MathQuaternionSwingTwist = &RigVMFunction_MathQuaternionSwingTwist(
    twistAxis=@n1.twistAxis
    Input=&RigUnit_GetRelativeTransformForItem(bChildInitial=False, bParentInitial=False, child=@n1.child, parent=@n1.parent).RelativeTransform.Rotation
   )
   n2 = RigVMDispatchNode(
    Values=((),())
    Index=@RigVMFunction_MathBoolToInteger.Result
    Values.0=@RigVMFunction_MathQuaternionSwingTwist.Swing
    Values.1=@RigVMFunction_MathQuaternionSwingTwist.Twist
   ) {template="DISPATCH_RigVMDispatch_SelectInt32(in Index,in Values,out Result)"}
   RigVMFunction_MathQuaternionToEuler = &RigVMFunction_MathQuaternionToEuler(Value=@n2.Result, RotationOrder=@n1.Rotation Order)
   > DISPATCH_RigVMDispatch_Print = RigVMDispatchNode(ScreenColor=(R=1,G=1,B=1,A=1), ScreenDuration=1.5, Enabled=@n1.debugRotation, Value=@RigVMFunction_MathQuaternionToEuler.Result) {template="DISPATCH_RigVMDispatch_Print(in Prefix,in Value,in Enabled,in ScreenDuration,in ScreenColor)"}
   Cast = RigVMDispatchNode(Value=@n1.triggerAxis) {template="DISPATCH_RigVMDispatch_CastEnumToInt(in Value,out Result)"}
   DISPATCH_RigVMDispatch_SelectInt32 = RigVMDispatchNode(
    Values=(0,0,0)
    Index=@Cast.Result
    Values.0=@RigVMFunction_MathQuaternionToEuler.Result.X
    Values.1=@RigVMFunction_MathQuaternionToEuler.Result.Y
    Values.2=@RigVMFunction_MathQuaternionToEuler.Result.Z
   ) {template="DISPATCH_RigVMDispatch_SelectInt32(in Index,in Values,out Result)"}
   Remap = &RigVMFunction_MathDoubleRemap(
    bClamp=False
    Value=@DISPATCH_RigVMDispatch_SelectInt32.Result
    TargetMinimum=@n1.outputMinMax.X
    TargetMaximum=@n1.outputMinMax.Y
    SourceMinimum=@n1.rotationMinMax.X
    SourceMaximum=@n1.rotationMinMax.Y
   )
   Absolute = &RigVMFunction_MathDoubleAbs(Value=@Remap.Result)
   Multiply = &RigVMFunction_MathDoubleMul(
    A=&RigVMFunction_MathDoubleClamp(Value=RigVMDispatchNode(True=@Absolute.Result, Condition=@n1.absoluteValue, False=@Remap.Result) {template="DISPATCH_RigVMDispatch_If(in Condition,in True,in False,out Result)"}.Result).Result
    B=@n1.premult
   )
   > RigVMFunctionReturnNode(outputQuat=@n2.Result, outputWeight=@Multiply.Result)
  ]}
 RigVMFunctionLibrary {nodes [
   DeltaRotation v4 = RigVMCollapseNode(
    ExecuteContext=()
    Rotation Order=ZYX
    child=(Type=Bone,Name="None")
    outputMinMax=(X=0,Y=1)
    parent=(Type=Bone,Name="None")
    premult=1
    rotationMinMax=(X=0,Y=0)
    twistAxis=(X=1,Y=0,Z=0)
   )
   arms_movement = RigVMCollapseNode(ExecuteContext=())
   head_tube = RigVMCollapseNode(ExecuteContext=())
  ]}
 RigVMModel {nodes [
   n1 = &RigUnit_PrepareForExecution()
   &RigUnit_HierarchyAddNull(
    Name=head_null
    Parent=(Type=Bone,Name="neck_01")
    Space=GlobalSpace
    Transform=&RigUnit_GetTransform(Item=(Type=Bone,Name="head"), bInitial=False).Transform
    ExecutePin=@n1.ExecutePin
   )
   > &RigUnit_HierarchyAddNull(
    Name=neck2_null
    Parent=(Type=Bone,Name="spine_05")
    Space=GlobalSpace
    Transform=&RigUnit_GetTransform(Item=(Type=Bone,Name="neck_02"), bInitial=False).Transform
   )
   > &RigUnit_HierarchyAddNull(
    Name=tube_pv_null
    Parent=(Type=Bone,Name="spine_05")
    Space=LocalSpace
    Transform=(Rotation=(X=0,Y=0,Z=0,W=1),Translation=(X=30,Y=-35,Z=18),Scale3D=(X=1,Y=1,Z=1))
   )
   ItemArray_1 = RigVMDispatchNode(Value=((Type=Bone,Name="headtube_01"),(Type=Bone,Name="headtube_02"),(Type=Bone,Name="headtube_03"),(Type=Bone,Name="headtube_04"))) {template="DISPATCH_RigVMDispatch_Constant(io Value)"}
   > For_Each = RigVMDispatchNode(Array=/Script/ControlRig.RigUnit_GetTransformItemArray(Space=GlobalSpace, bInitial=False, Items=@ItemArray_1.Value) {is=struct RigVMUnitNode}.Transforms) {template="DISPATCH_RigVMDispatch_ArrayIterator(in Array,out Element,out Index,out Count,out Ratio)"}
   Replace = /Script/RigVM.RigVMFunction_NameReplace(
    New=tube_null
    Old=tube
    Name=RigVMDispatchNode(Index=@For_Each.Index, Array=@ItemArray_1.Value) {template="DISPATCH_RigVMDispatch_ArrayGetAtIndex(in Array,in Index,out Element)"}.Element.Name
   ) {is=struct RigVMUnitNode, template="Replace::Execute(in Name,in Old,in New,out Result)"}
   > &RigUnit_HierarchyAddNull(Parent=(Type=Bone,Name="seat_base"), Space=GlobalSpace, Transform=@For_Each.Element, Name=@Replace.Result)
   /Script/ControlRig.RigUnit_SwitchParent(
    Child=(Type=Null,Name="headtube_null_02")
    Mode=ParentItem
    Parent=(Type=Bone,Name="neck_01")
    bMaintainGlobal=True
    ExecutePin=@For_Each.Completed
   ) {is=struct RigVMUnitNode}
   > &RigUnit_SwitchParent(Child=(Type=Null,Name="headtube_null_03"), Mode=ParentItem, Parent=(Type=Bone,Name="head"), bMaintainGlobal=True)
   > &RigUnit_SwitchParent(Child=(Type=Null,Name="headtube_null_04"), Mode=ParentItem, Parent=(Type=Bone,Name="head"), bMaintainGlobal=True)
   RigUnit_BeginExecution = &RigUnit_BeginExecution()
   FN_DeformRig_FNC = RigVMFunctionReferenceNode(
    DeformRig_ Curve="(deformRig_l_upperarm_up_alpha_72_#1=true,deformRig_r_upperarm_up_alpha_73_#2=true,deformRig_l_upperarm_fr_alpha_74_#3=true,deformRig_r_upperarm_fr_alpha_75_CB51B80C4 ...(truncated)"
    DeformRig_Bool="(upperarm_1_#4=true,elbow_3_#5=false,wrist_7_#6=false,groin_8_#7=false,glute_17_#8=false,knee_9_40C408324C ...(truncated)"
    DeformRig_Debug=(upperarm_debug_18_#4=False,wrist_debug_20_#6=False,lat_debug_26_#9=False,groinGlute_debug_27_#7=False)
    DeformRig_Value="(upperarm_up_18_#10=(Rotation=(X=0.0,Y=-0.0,Z=0.0,W=1.0),Translation=(X=0.0,Y=0.0,Z=0),Scale3D=(X=1,Y=1.0,Z=1.0)),upperarm_fwd_fr_19_#11=(Rotation=(X=0.0,Y=-0.0,Z=0.0,W=1.0),Translation=(X=0.0,Y=0,Z ...(truncated)"
    ExecuteContext=@RigUnit_BeginExecution.ExecutePin
   )
   > &RigUnit_ModifyTransforms(
    ItemToModify="((Item=(Type=Bone,Name=\"arm_01_l\"),Transform=(Rotation=(X=0.1093,Y=0.2877,Z=0.2531,W=0.9172),Translation=(X=0.4361,Y=-0.3864,Z=0.2887),Scale3D=(X=1,Y=1,Z=1))),(Item=(Type=Bone,Name=\"arm_01_r\"),Transform=(Rotation=(X=0.1419,Y=0.3066,Z=0.2057,W=0.9185),Trans ...(truncated)"
    Weight=1
    WeightMinimum=0
   )
   > &RigUnit_ModifyTransforms(
    ItemToModify="((Item=(Type=Bone,Name=\"lil_thumb_01_l\"),Transform=(Rotation=(X=-0.0603,Y=0.2105,Z=-0.2154,W=0.9517),Translation=(X=0.077,Y=0.0705,Z=0.4),Scale3D=(X=1,Y=1.1,Z=1))),(Item=(Type=Bone,Name=\"lil_thumb_02_l\"),Transform=(Rotation=(X=-0.0059,Y=0.5373,Z=0.0033,W=0 ...(truncated)"
    Weight=1
    WeightMinimum=0
   )
   > &RigUnit_ModifyTransforms(
    ItemToModify="((Item=(Type=Bone,Name=\"lil_thumb_01_r\"),Transform=(Rotation=(X=-0.0603,Y=0.2105,Z=-0.2154,W=0.9517),Translation=(X=-0.077,Y=-0.0705,Z=-0.4),Scale3D=(X=1,Y=1.1,Z=1))),(Item=(Type=Bone,Name=\"lil_thumb_02_r\"),Transform=(Rotation=(X=-0.0059,Y=0.5373,Z=0.0033, ...(truncated)"
    Weight=1
    WeightMinimum=0
   )
   > &RigUnit_ParentConstraint(
    AdvancedSettings=(InterpolationType=Average,RotationOrderForFilter=XZY)
    Child=(Type=Bone,Name="head_transition")
    Filter=(TranslationFilter=(bX=true,bY=true,bZ=true),RotationFilter=(bX=false,bY=True,bZ=True),ScaleFilter=(bX=True,bY=True,bZ=True))
    Parents=((Item=(Type=Bone,Name="head"),Weight=1))
    Weight=1
   )
   > &RigUnit_ParentConstraint(
    AdvancedSettings=(InterpolationType=Average,RotationOrderForFilter=XZY)
    Child=(Type=Bone,Name="seatbelt_C")
    Parents=((Item=(Type=Bone,Name="neck_01"),Weight=0.25),(Item=(Type=Bone,Name="neck_02"),Weight=0.75))
    Weight=1
   )
   > arms_movement = RigVMFunctionReferenceNode()
   > head_tube = RigVMFunctionReferenceNode()
   > DeltaRotation v4_4 = RigVMFunctionReferenceNode(
    Rotation Order=ZYX
    child=(Type=Bone,Name="head")
    outputMinMax=(X=0,Y=1)
    parent=(Type=Null,Name="head_null")
    premult=1
    rotationMinMax=(X=0,Y=-20)
    triggerAxis=Z
    twistAxis=(X=1,Y=0,Z=0)
   )
   > &RigUnit_ModifyTransforms(
    ItemToModify="((Item=(Type=Bone,Name=\"seat\"),Transform=(Rotation=(X=0,Y=0,Z=0.3319,W=-0.9433),Translation=(X=-2.7839,Y=-1.6724,Z=0),Scale3D=(X=1,Y=1,Z=1))),(Item=(Type=Bone,Name=\"goggles_ft\"),Transform=(Rotation=(X=0,Y=0,Z=0,W=-1),Translation=(X=-0.3418,Y=0,Z=0),Scale3D ...(truncated)"
    WeightMinimum=0
    Weight=@"DeltaRotation v4_4".outputWeight
   )
   > DeltaRotation v4_1 = RigVMFunctionReferenceNode(
    Rotation Order=ZYX
    child=(Type=Bone,Name="neck_02")
    outputMinMax=(X=0,Y=1)
    parent=(Type=Null,Name="neck2_null")
    premult=1
    rotationMinMax=(X=0,Y=-35)
    triggerAxis=Z
    twistAxis=(X=1,Y=0,Z=0)
   )
   > &RigUnit_ModifyTransforms(
    ItemToModify=((Item=(Type=Bone,Name="seat_base"),Transform=(Rotation=(X=0,Y=0,Z=-0.1305,W=0.9914),Translation=(X=0,Y=0,Z=0),Scale3D=(X=1,Y=1,Z=1))))
    WeightMinimum=0
    Weight=@"DeltaRotation v4_1".outputWeight
   )
   > DeltaRotation v4_2 = RigVMFunctionReferenceNode(
    Rotation Order=ZYX
    child=(Type=Bone,Name="head")
    onlyTwist=true
    outputMinMax=(X=0,Y=1)
    parent=(Type=Null,Name="head_null")
    premult=1
    rotationMinMax=(X=0,Y=30)
    triggerAxis=X
    twistAxis=(X=1,Y=0,Z=0)
   )
   > &RigUnit_ModifyTransforms(
    ItemToModify="((Item=(Type=Bone,Name=\"goggle_bk\"),Transform=(Rotation=(X=0.1728,Y=0,Z=0,W=-0.985),Translation=(X=0,Y=0.1041,Z=4.3675),Scale3D=(X=1,Y=1,Z=1))),(Item=(Type=Bone,Name=\"goggles_ft\"),Transform=(Rotation=(X=-0.3852,Y=-0.179,Z=-0.0376,W=-0.9045),Translation=(X= ...(truncated)"
    WeightMinimum=0
    Weight=@"DeltaRotation v4_2".outputWeight
   )
   > DeltaRotation v4_5 = RigVMFunctionReferenceNode(
    Rotation Order=ZYX
    child=(Type=Bone,Name="head")
    onlyTwist=true
    outputMinMax=(X=0,Y=1)
    parent=(Type=Null,Name="head_null")
    premult=1
    rotationMinMax=(X=30,Y=60)
    triggerAxis=X
    twistAxis=(X=1,Y=0,Z=0)
   )
   > &RigUnit_ModifyTransforms(
    ItemToModify="((Item=(Type=Bone,Name=\"goggle_bk\"),Transform=(Rotation=(X=-0.0518,Y=0.1182,Z=0.0062,W=0.9916),Translation=(X=0.9476,Y=0,Z=2.9254),Scale3D=(X=1,Y=1,Z=1))),(Item=(Type=Bone,Name=\"goggles_ft\"),Transform=(Rotation=(X=0.5702,Y=0.0943,Z=-0.0614,W=0.8137),Transl ...(truncated)"
    WeightMinimum=0
    Weight=@"DeltaRotation v4_5".outputWeight
   )
   > DeltaRotation v4_6 = RigVMFunctionReferenceNode(
    Rotation Order=ZYX
    child=(Type=Bone,Name="head")
    onlyTwist=true
    outputMinMax=(X=0,Y=1)
    parent=(Type=Null,Name="head_null")
    premult=1
    rotationMinMax=(X=0,Y=-30)
    triggerAxis=X
    twistAxis=(X=1,Y=0,Z=0)
   )
   > &RigUnit_ModifyTransforms(
    ItemToModify="((Item=(Type=Bone,Name=\"goggle_bk\"),Transform=(Rotation=(X=0.0712,Y=-0.0909,Z=-0.0267,W=0.9929),Translation=(X=0,Y=0,Z=-1.25),Scale3D=(X=1,Y=1,Z=1))),(Item=(Type=Bone,Name=\"goggles_ft\"),Transform=(Rotation=(X=0.1491,Y=0.0998,Z=-0.0215,W=-0.9835),Translatio ...(truncated)"
    WeightMinimum=0
    Weight=@"DeltaRotation v4_6".outputWeight
   )
   > DeltaRotation v4_3 = RigVMFunctionReferenceNode(
    Rotation Order=ZYX
    child=(Type=Bone,Name="head")
    onlyTwist=true
    outputMinMax=(X=0,Y=1)
    parent=(Type=Null,Name="head_null")
    premult=1
    rotationMinMax=(X=-30,Y=-60)
    triggerAxis=X
    twistAxis=(X=1,Y=0,Z=0)
   )
   > &RigUnit_ModifyTransforms(
    ItemToModify="((Item=(Type=Bone,Name=\"goggle_bk\"),Transform=(Rotation=(X=0.1742,Y=-0.4284,Z=-0.0456,W=0.8855),Translation=(X=3.1435,Y=0.623,Z=-1.6728),Scale3D=(X=1,Y=1,Z=1))),(Item=(Type=Bone,Name=\"goggles_ft\"),Transform=(Rotation=(X=-0.1027,Y=-0.0946,Z=0.2858,W=0.948), ...(truncated)"
    WeightMinimum=0
    Weight=@"DeltaRotation v4_3".outputWeight
   )
   > DeltaRotation v4_7 = RigVMFunctionReferenceNode(
    Rotation Order=ZYX
    child=(Type=Bone,Name="neck_02")
    onlyTwist=true
    outputMinMax=(X=0,Y=1)
    parent=(Type=Null,Name="neck2_null")
    premult=1
    rotationMinMax=(X=20,Y=60)
    triggerAxis=X
    twistAxis=(X=1,Y=0,Z=0)
   )
   > &RigUnit_ModifyTransforms(
    ItemToModify="((Item=(Type=Bone,Name=\"seatbelt_A\"),Transform=(Rotation=(X=-0.0551,Y=-0.0874,Z=0.0163,W=0.9945),Translation=(X=0,Y=0,Z=-1.5888),Scale3D=(X=1,Y=1,Z=1))),(Item=(Type=Bone,Name=\"seatbelt_B\"),Transform=(Rotation=(X=0.0006,Y=0.0125,Z=0.0499,W=-0.9987),Translat ...(truncated)"
    WeightMinimum=0
    Weight=@"DeltaRotation v4_7".outputWeight
   )
   > DeltaRotation v4_8 = RigVMFunctionReferenceNode(
    Rotation Order=ZYX
    child=(Type=Bone,Name="neck_02")
    outputMinMax=(X=0,Y=1)
    parent=(Type=Null,Name="neck2_null")
    premult=1
    rotationMinMax=(X=0,Y=35)
    triggerAxis=Z
    twistAxis=(X=1,Y=0,Z=0)
   )
   > &RigUnit_ModifyTransforms(
    ItemToModify=((Item=(Type=Bone,Name="seatbelt_C"),Transform=(Rotation=(X=0,Y=0,Z=0,W=1),Translation=(X=2.5,Y=0,Z=0),Scale3D=(X=1,Y=1,Z=1))))
    WeightMinimum=0
    Weight=@"DeltaRotation v4_8".outputWeight
   )
   > DeltaRotation v4_9 = RigVMFunctionReferenceNode(
    Rotation Order=ZYX
    child=(Type=Bone,Name="head")
    outputMinMax=(X=0,Y=1)
    parent=(Type=Null,Name="head_null")
    premult=1
    rotationMinMax=(X=0,Y=15)
    triggerAxis=Z
    twistAxis=(X=1,Y=0,Z=0)
   )
   > &RigUnit_ModifyTransforms(
    ItemToModify="((Item=(Type=Bone,Name=\"seatbelt_B\"),Transform=(Rotation=(X=0,Y=0,Z=0,W=1),Translation=(X=0.5,Y=0,Z=-1),Scale3D=(X=1,Y=1,Z=1))),(Item=(Type=Bone,Name=\"seatbelt_A\"),Transform=(Rotation=(X=0,Y=0,Z=0,W=-1),Translation=(X=0.5,Y=0,Z=-1.8),Scale3D=(X=1,Y=1,Z=1)) ...(truncated)"
    WeightMinimum=0
    Weight=@"DeltaRotation v4_9".outputWeight
   )
   > DeltaRotation v4_10 = RigVMFunctionReferenceNode(
    Rotation Order=ZYX
    child=(Type=Bone,Name="head")
    outputMinMax=(X=0,Y=1)
    parent=(Type=Null,Name="head_null")
    premult=1
    rotationMinMax=(X=0,Y=5)
    triggerAxis=Y
    twistAxis=(X=1,Y=0,Z=0)
   )
   > &RigUnit_ModifyTransforms(
    ItemToModify="((Item=(Type=Bone,Name=\"seatbelt_B\"),Transform=(Rotation=(X=0,Y=0,Z=0,W=1),Translation=(X=0,Y=0,Z=-0.25),Scale3D=(X=1,Y=1,Z=1))),(Item=(Type=Bone,Name=\"seatbelt_A\"),Transform=(Rotation=(X=0,Y=0,Z=0,W=-1),Translation=(X=0.25,Y=0,Z=-1),Scale3D=(X=1,Y=1,Z=1)) ...(truncated)"
    WeightMinimum=0
    Weight=@"DeltaRotation v4_10".outputWeight
   )
  ]}
]
hierarchy [
 root {loc=<0 0 0>, rot=<0 0 0 1>}
 pelvis {parent=root, loc=<0 -0.7321 92.4218>, rot=<0.0127 0.707 -0.0127 -0.707>}
 spine_01 {parent=pelvis, loc=<5.6002 -0.5128 0>, rot=<0 0 0.034 -0.9994>}
 spine_02 {parent=spine_01, loc=<8.0756 0.0139 0>, rot=<0 0 0.0238 -0.9997>}
 spine_03 {parent=spine_02, loc=<7.0395 0 0>, rot=<0 0 0.0149 -0.9999>}
 spine_04 {parent=spine_03, loc=<8.1023 0 0>, rot=<0 0 -0.0132 -0.9999>}
 spine_05 {parent=spine_04, loc=<10.0141 0 0>, rot=<0 0 -0.0015 -1>}
 clavicle_l {parent=spine_05, loc=<2.5856 -0.2866 -4.3777>, rot=<0.0465 -0.7289 -0.009 -0.6829>}
 upperarm_l {parent=clavicle_l, loc=<18.2264 0.1342 -0.001>, rot=<0.0253 -0.2693 0.0013 -0.9627>}
 lowerarm_l {parent=upperarm_l, loc=<29.0071 0.1686 -1.239>, rot=<0.0391 -0.0099 -0.2453 -0.9686>}
 hand_l {parent=lowerarm_l, loc=<25.8607 0.222 1.0274>, rot=<-0.7193 -0.0988 0.1193 -0.6773>}
 elbow_sphere_l {parent=lowerarm_l, loc=<-2.1071 0.1869 0.1539>, rot=<0 0 0 1>}
 robot_lowerarm_l {parent=lowerarm_l, loc=<-2.1071 0.1869 0.1539>, rot=<-0.0473 0.0161 -0.001 -0.9987>}
 robot_wrist_l {parent=robot_lowerarm_l, loc=<27.9815 0.0185 -0.0305>, rot=<0 0 0 1>}
 robot_upperarm_l {parent=upperarm_l, loc=<3.3767 0.2443 0.8522>, rot=<-0.0065 -0.0392 0.0192 -0.999>}
 neck_01 {parent=spine_05, loc=<10.815 1.9196 0>, rot=<0 0 -0.1864 -0.9825>}
 neck_02 {parent=neck_01, loc=<4.7347 0.3045 0>, rot=<0 0 0.0363 -0.9993>}
 head {parent=neck_02, loc=<5.7632 0.5552 0>, rot=<0 0 0.115 -0.9934>}
 MSTR_pompom {parent=head, loc=<24.7988 -0.6773 -0.012>, rot=<0 0 0 1>}
 dyn_pompom {parent=MSTR_pompom, loc=<0 0 0>, rot=<0 0 0 1>}
 goggle {parent=head, loc=<7.0922 18.1949 4.337>, rot=<0.517 0.4768 0.6273 0.3346>}
 base_arm_01_l {parent=neck_02, loc=<1.5765 2.7523 -10.721>, rot=<-0.126 -0.8545 0.0382 -0.5025>}
 arm_01_l {parent=base_arm_01_l, loc=<0 0 0>, rot=<0 0 0 1>}
 arm_02_l {parent=arm_01_l, loc=<5.6988 0 0>, rot=<0 0 0 1>}
 lil_thumb_01_l {parent=arm_02_l, loc=<0.7064 1.3908 -0.9273>, rot=<0.6746 -0.2023 -0.3869 -0.5953>}
 lil_thumb_02_l {parent=lil_thumb_01_l, loc=<1.8313 0 0>, rot=<-0.0027 -0.1169 -0.0596 -0.9914>}
 lil_index_base_l {parent=arm_02_l, loc=<1.5687 1.4794 0.0335>, rot=<0.0861 -0.1066 -0.0562 -0.989>}
 lil_index_01_l {parent=lil_index_base_l, loc=<2.3756 0 0>, rot=<-0.0281 -0.2511 -0.0031 -0.9676>}
 lil_index_02_l {parent=lil_index_01_l, loc=<0.8996 0 0>, rot=<0.0031 -0.2056 0.0392 -0.9778>}
 lil_index_03_l {parent=lil_index_02_l, loc=<0.7081 0 0>, rot=<0.0202 -0.1701 -0.0122 -0.9851>}
 lil_middle_base_l {parent=arm_02_l, loc=<1.7799 0.7248 0.2805>, rot=<0.0877 -0.0916 -0.0363 -0.9913>}
 lil_middle_01_l {parent=lil_middle_base_l, loc=<2.3756 0 0>, rot=<-0.0282 -0.2129 -0.002 -0.9767>}
 lil_middle_02_l {parent=lil_middle_01_l, loc=<0.8996 0 0>, rot=<-0.0074 -0.2446 0.0042 -0.9696>}
 lil_middle_03_l {parent=lil_middle_02_l, loc=<0.8494 0.0034 -0.0272>, rot=<0.0192 -0.0947 -0.0137 -0.9952>}
 lil_ring_base_l {parent=arm_02_l, loc=<1.8313 -0.3108 0.4285>, rot=<0.0231 -0.1209 -0.0156 -0.9923>}
 lil_ring_01_l {parent=lil_ring_base_l, loc=<2.3756 0 0>, rot=<-0.0281 -0.2511 -0.0031 -0.9676>}
 lil_ring_02_l {parent=lil_ring_01_l, loc=<0.8996 0 0>, rot=<-0.0004 -0.1335 -0.0079 -0.991>}
 lil_ring_03_l {parent=lil_ring_02_l, loc=<0.7081 0 0>, rot=<0.0298 -0.1876 0.037 -0.9811>}
 lil_pinky_base_l {parent=arm_02_l, loc=<1.7324 -0.8752 0.3448>, rot=<0.0309 -0.1192 0.0483 -0.9912>}
 lil_pinky_01_l {parent=lil_pinky_base_l, loc=<2.1212 -0.0065 0.0387>, rot=<-0.0358 -0.2377 -0.0306 -0.9702>}
 lil_pinky_02_l {parent=lil_pinky_01_l, loc=<0.9066 -0.0013 0.0338>, rot=<0.0002 -0.2044 -0.0079 -0.9789>}
 lil_pinky_03_l {parent=lil_pinky_02_l, loc=<0.7104 0.0022 0.0198>, rot=<0.0294 -0.1187 0.0099 -0.9925>}
 armwire_l {parent=arm_01_l, loc=<2.8212 1.8094 -3.1108>, rot=<0.2938 -0.0264 -0.5692 -0.7675>}
 seatbelt_A {parent=neck_02, loc=<2.4914 8.0708 -8.3673>, rot=<-0.2579 -0.4665 -0.3106 -0.787>}
 base_arm_01_r {parent=neck_02, loc=<1.5762 2.7524 10.721>, rot=<0.8545 -0.126 -0.5025 -0.0382>}
 arm_01_r {parent=base_arm_01_r, loc=<0 0 0>, rot=<0 0 0 1>}
 arm_02_r {parent=arm_01_r, loc=<-5.6985 0 -0.0007>, rot=<0 0 0 1>}
 lil_thumb_01_r {parent=arm_02_r, loc=<-0.7065 -1.3908 0.9275>, rot=<0.6746 -0.2023 -0.3869 -0.5953>}
 lil_thumb_02_r {parent=lil_thumb_01_r, loc=<-1.8309 0 -0.0001>, rot=<-0.0027 -0.1169 -0.0596 -0.9914>}
 lil_index_base_r {parent=arm_02_r, loc=<-1.569 -1.4794 -0.033>, rot=<0.0861 -0.1066 -0.0562 -0.989>}
 lil_index_01_r {parent=lil_index_base_r, loc=<-2.3752 0 -0.0004>, rot=<-0.0281 -0.2511 -0.0031 -0.9676>}
 lil_index_02_r {parent=lil_index_01_r, loc=<-0.9 0 0.0003>, rot=<0.0031 -0.2056 0.0392 -0.9778>}
 lil_index_03_r {parent=lil_index_02_r, loc=<-0.7076 0 0>, rot=<0.0202 -0.1701 -0.0122 -0.9851>}
 lil_middle_base_r {parent=arm_02_r, loc=<-1.7803 -0.7248 -0.2798>, rot=<0.0877 -0.0916 -0.0363 -0.9913>}
 lil_middle_01_r {parent=lil_middle_base_r, loc=<-2.3753 0 -0.0004>, rot=<-0.0282 -0.2129 -0.002 -0.9767>}
 lil_middle_02_r {parent=lil_middle_01_r, loc=<-0.8994 0 -0.0001>, rot=<-0.0074 -0.2446 0.0042 -0.9696>}
 lil_middle_03_r {parent=lil_middle_02_r, loc=<-0.8499 -0.0034 0.0271>, rot=<0.0192 -0.0947 -0.0137 -0.9952>}
 lil_ring_base_r {parent=arm_02_r, loc=<-1.8315 0.3108 -0.428>, rot=<0.0231 -0.1209 -0.0156 -0.9923>}
 lil_ring_01_r {parent=lil_ring_base_r, loc=<-2.3756 0 0>, rot=<-0.0281 -0.2511 -0.0031 -0.9676>}
 lil_ring_02_r {parent=lil_ring_01_r, loc=<-0.8992 0 -0.0002>, rot=<-0.0004 -0.1335 -0.0079 -0.991>}
 lil_ring_03_r {parent=lil_ring_02_r, loc=<-0.7081 0 0.0001>, rot=<0.0298 -0.1876 0.037 -0.9811>}
 lil_pinky_base_r {parent=arm_02_r, loc=<-1.7327 0.8752 -0.3442>, rot=<0.0309 -0.1192 0.0483 -0.9912>}
 lil_pinky_01_r {parent=lil_pinky_base_r, loc=<-2.121 0.0065 -0.039>, rot=<-0.0358 -0.2377 -0.0306 -0.9702>}
 lil_pinky_02_r {parent=lil_pinky_01_r, loc=<-0.907 0.0013 -0.0337>, rot=<0.0002 -0.2044 -0.0079 -0.9789>}
 lil_pinky_03_r {parent=lil_pinky_02_r, loc=<-0.7099 -0.0023 -0.0197>, rot=<0.0294 -0.1187 0.0099 -0.9925>}
 armwire_r {parent=arm_01_r, loc=<0.3112 4.0001 -4.2449>, rot=<0.3426 -0.4624 -0.1765 -0.7986>}
 seatbelt_B {parent=neck_02, loc=<3.7996 3.0568 -12.4387>, rot=<-0.0962 -0.4223 -0.0181 -0.9012>}
 head_transition {parent=neck_02, loc=<0 0 0>, rot=<0 0 0 1>}
 goggle_bk {parent=head_transition, loc=<2.3871 -3.2246 4.8933>, rot=<-0.5718 -0.7999 0.1816 0.0156>}
 goggles_ft {parent=head_transition, loc=<3.6331 7.6843 6.7659>, rot=<-0.2317 -0.1739 0.3117 -0.905>}
 seatbelt_C {parent=neck_02, loc=<-0.5113 11.2265 -1.5734>, rot=<-0.4221 -0.3275 -0.5572 -0.6358>}
 clavicle_r {parent=spine_05, loc=<2.5855 -0.2866 4.3777>, rot=<0.7289 0.0465 -0.6829 0.009>}
 upperarm_r {parent=clavicle_r, loc=<-18.2264 -0.1342 0.001>, rot=<0.0253 -0.2693 0.0013 -0.9627>}
 lowerarm_r {parent=upperarm_r, loc=<-29.0071 -0.1686 1.239>, rot=<0.0391 -0.0099 -0.2453 -0.9686>}
 hand_r {parent=lowerarm_r, loc=<-25.8607 -0.222 -1.0274>, rot=<-0.7193 -0.0988 0.1193 -0.6773>}
 robot_upperarm_r {parent=upperarm_r, loc=<-3.3769 -0.2443 -0.8518>, rot=<-0.0065 -0.0392 0.0192 -0.999>}
 seat_base {parent=spine_05, loc=<8.6918 -6.9623 -0.0515>, rot=<0 0 0 1>}
 seat {parent=seat_base, loc=<13.888 -2.5097 0.3267>, rot=<0 0 0 1>}
 headtube_01 {parent=seat_base, loc=<10.7266 -6.9836 0.6024>, rot=<-0.4891 0.4704 0.1586 -0.7172>}
 headtube_02 {parent=headtube_01, loc=<9.0781 0 0>, rot=<0.0134 -0.4978 0.0198 -0.867>}
 headtube_03 {parent=headtube_02, loc=<10.1989 0 0>, rot=<-0.0347 -0.5057 0.013 -0.8619>}
 headtube_04 {parent=headtube_03, loc=<12.3015 0 0>, rot=<0.0171 0 0 -0.9999>}
 dyn_headtube_04 {parent=headtube_04, loc=<0 0 0>, rot=<0 0 0 1>}
 dyn_headtube_03 {parent=headtube_03, loc=<0 0 0>, rot=<0 0 0 1>}
 dyn_headtube_02 {parent=headtube_02, loc=<0 0 0>, rot=<0 0 0 1>}
 dyn_headtube_01 {parent=headtube_01, loc=<0 0 0>, rot=<0 0 0 1>}
]
variables [
 Character Backward {type=double, default=0}
 Character Forward {type=double, default=0}
 Character Move Left {type=double, default=0}
 Character Move Right {type=double, default=0}
 jumping {type=bool, default=False}
 meleeanim {type=bool, default=False}
 targeting {type=bool, default=False}
]
