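# Animation Blueprint for the LiquidCouch large male skin's face accessory, adding procedural and secondary motion to copied character poses.
# Parent-mesh pose, blended with Body-mesh pose from both clavicles, feeds a Control Rig driven by movement, targeting, jumping and melee state.
# Anim Dynamics then lightly simulates the dyn_pompom bone with an angular spring.
# EventGraph updates movement values, plus gravity and rigid-body settings that the AnimGraph never reads.
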AnimBlueprint /BRCosmetics/Characters/Player/Male/Large/Bodies/M_LRG_LiquidCouch/Meshes/Parts/M_LRG_LiquidCouch_FaceAcc_AnimBP
target_skeleton=&/M_LRG_LiquidCouch_FaceAcc_Skeleton
graphs [
 AnimGraph {nodes [
   n1 = Get Body() -> (Body:Skeletal Mesh Component Budgeted Object Reference)
   n2 = Control Rig!(
    Source=Layered blend per bone(
     BasePose=Copy Pose From Mesh() {bUseAttachedParent=true}
     BlendPoses_0=Copy Pose From Mesh(SourceMeshComponent=@n1) {bUseAttachedParent=true}
    ) {layerSetup [{branchFilters [{boneName=clavicle_l}, {boneName=clavicle_r}]}]}
    Alpha
    Character Forward:double=Get Character Forward() -> (Character Forward:double)
    Character Backward:double=Get Character Backward() -> (Character Backward:double)
    Character Move Right:double=Get Character Move Right() -> (Character Move Right:double)
    Character Move Left:double=Get Character Move Left() -> (Character Move Left:double)
    targeting:bool=Get bIsTargeting()
    jumping:bool=Get bIsJumping()
    meleeanim:bool=Get bIsPlayingMeleeAnim()
   ) -> (Pose) {controlRigAssetReference {blueprintRigClass=&/M_LRG_LiquidCouch_FaceAcc_CtrlRig.M_LRG_LiquidCouch_FaceAcc_CtrlRig_C}}
   n3 = Anim Dynamics - Bone: dyn_pompom!(ComponentPose=Local To Component(LocalPose=@n2), Alpha=0.25, GravityOverride:Vector="0,0,-980", ExternalForce="0,0,0") -> (Pose) {
    angularDampingOverride=0.9999
    angularSpringConstant=8000
    bAngularSpring=true
    bOverrideAngularDamping=true
    bUseGravityOverride=true
    bUsePlanarLimit=false
    gravityScale=0.35
    simulationSpace=RootRelative
    alphaBoolBlend {blendOption=Linear}
    boundBone {boneName=dyn_pompom}
    componentAppliedLinearAccClamp=<1400 1400 1400>
    componentLinearAccScale=<1 1 1>
    componentLinearVelScale=<1 1 1>
    gravityOverride {z=-980}
    physicsBodyDefinitions [{
      boundBone {boneName=dyn_pompom}
      boxExtents=<20 15 10>
      constraintSetup {angularLimitsMax=<0 45 45>, angularLimitsMin=<0 -45 -45>, angularTarget=<1 0 0>}
      localJointOffset=<4 0 1>
     }]
   }
   Output Pose(Result=Component To Local(ComponentPose=@n3)) {layerGroup=None}
  ]}
 EventGraph {nodes [
   Event BlueprintInitializeAnimation()
   > n1 = Cast To FortPlayerPawn(Object=TryGetFortPawn())
   > Set Pawn(Pawn:Fort Player Pawn Object Reference=@n1) -> (Output_Get:Fort Player Pawn Object Reference)
   Event BlueprintUpdateAnimation()
   > n2 = "Gravity Override State Function (Parameter Asset) (Nativized)"(
    CharacterPartAnimInstance=Self-Reference()
    GameTimeInSeconds=GetGameTimeInSeconds()
    DebugJointName=spine_05
    Parameters=/BRCosmetics/Characters/Player/Male/Large/Bodies/M_LRG_LiquidCouch/Meshes/M_LRG_LiquidCouch_Pack_Parameters
   )
   > Set Gravity Override(Gravity Override=@n2)
   > Set RBAN Settings Out(RBAN Settings Out=@n2.OutRigidBodyAnimNodeParameters)
   > Get!(execute:Exec) -> (then:Exec, else:Exec, Pawn:Fort Player Pawn Object Reference)
   > n3 = PlayerMovement(AnimInstance=Self-Reference(), XDotProductCoefficient=25, YDotProductCoefficient=25, ZDotProductCoefficient=1)
   > Set Character Move Right(Character Move Right:double=MapRangeClamped(Value=@n3, InRangeA=0, InRangeB=25)) -> (Output_Get:double)
   > Set Character Move Left(Character Move Left:double=MapRangeClamped(Value=@n3, InRangeA=0, InRangeB=-25)) -> (Output_Get:double)
   > Set Character Forward(Character Forward:double=MapRangeClamped(Value=@n3.YMovement, InRangeA=0, InRangeB=25)) -> (Output_Get:double)
   > Set Character Backward(Character Backward:double=MapRangeClamped(Value=@n3.YMovement, InRangeA=0, InRangeB=-25)) -> (Output_Get:double)
   n4 = Event OnCharacterCustomizationLoaded()
   > Set Body(Body:Skeletal Mesh Component Budgeted Object Reference=Get Array Item with EFort Custom Part Type(Array=@n4.CustomMeshParts)) -> (Output_Get:Skeletal Mesh Component Budgeted Object Reference)
  ]}
]
variables [
 Body {type=SkeletalMeshComponentBudgeted}
 Character Backward {type=number}
 Character Forward {type=number}
 Character Move Left {type=number}
 Character Move Right {type=number}
 Gravity Override {type=Vector}
 Pawn {type=FortPlayerPawn}
 RBAN Settings Out {
  type=FortRigidBodyAnimNodeParameters
  default {simSpaceSettings {dampingAlpha=0, maxAngularAcceleration=16, maxAngularVelocity=4, maxLinearAcceleration=800, maxLinearVelocity=200, velocityScaleZ=0.5}}
 }
]
